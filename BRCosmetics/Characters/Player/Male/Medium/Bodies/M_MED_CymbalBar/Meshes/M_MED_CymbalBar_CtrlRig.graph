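# Control Rig for the M_MED_CymbalBar male body that drives corrective and accessory bones for feet, armpits, belts, pockets and holster.
# A shared DeformRig function runs first, followed by collapsed per-region stages.
# Weighted parent constraints place belt and pocket bones, then spherical pose readers on limbs and spine weight additional offsets.
# The shoulder belt also uses bone-distance remaps and Kalman smoothing, and an "is moving" variable caps one pocket offset.
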ControlRigBlueprint /BRCosmetics/Characters/Player/Male/Medium/Bodies/M_MED_CymbalBar/Meshes/M_MED_CymbalBar_CtrlRig
graphs [
 CollapseNode_ContainedGraph {nodes [
   RigVMFunctionEntryNode()
   > n1 = &RigUnit_SphericalPoseReader(
    ActiveRegionScaleFactors=(PositiveWidth=1,NegativeWidth=1,PositiveHeight=1,NegativeHeight=1)
    ActiveRegionSize=0.15
    Debug=(bDrawDebug=false,bDraw2D=False,bDrawLocalAxes=False,DebugScale=25,DebugSegments=20,DebugThickness=0.15)
    DriverAxis=(X=1,Y=0,Z=0)
    DriverItem=(Type=Bone,Name="foot_r")
    FalloffRegionScaleFactors=(PositiveWidth=1,NegativeWidth=1,PositiveHeight=1,NegativeHeight=1)
    FalloffSize=0.5
    FlipHeightScaling=False
    FlipWidthScaling=False
    OptionalParentItem=(Type=Bone,Name="calf_r")
    RotationOffset=(X=90,Y=0,Z=10)
   )
   > &RigUnit_OffsetTransformForItem(
    Item=(Type=Bone,Name="foot_r_deform_base")
    OffsetTransform=(Rotation=(X=0,Y=-0.4226,Z=0,W=0.9063),Translation=(X=0,Y=0,Z=0),Scale3D=(X=1,Y=1,Z=1))
    bPropagateToChildren=True
    Weight=@n1.OutputParam
   )
   > &RigUnit_OffsetTransformForItem(
    Item=(Type=Bone,Name="foot_r_deform_back")
    OffsetTransform=(Rotation=(X=0,Y=0,Z=0,W=1),Translation=(X=11,Y=0,Z=0),Scale3D=(X=1,Y=1,Z=1))
    bPropagateToChildren=True
    Weight=@n1.OutputParam
   )
   > &RigUnit_OffsetTransformForItem(
    Item=(Type=Bone,Name="foot_r_deform_for")
    OffsetTransform=(Rotation=(X=0,Y=0,Z=0,W=1),Translation=(X=-5,Y=0,Z=0),Scale3D=(X=1,Y=1,Z=1))
    bPropagateToChildren=True
    Weight=@n1.OutputParam
   )
   > n2 = &RigUnit_SphericalPoseReader(
    ActiveRegionScaleFactors=(PositiveWidth=1,NegativeWidth=1,PositiveHeight=1,NegativeHeight=1)
    ActiveRegionSize=0.15
    Debug=(bDrawDebug=false,bDraw2D=False,bDrawLocalAxes=False,DebugScale=25,DebugSegments=20,DebugThickness=0.15)
    DriverAxis=(X=1,Y=0,Z=0)
    DriverItem=(Type=Bone,Name="foot_r")
    FalloffRegionScaleFactors=(PositiveWidth=1,NegativeWidth=1,PositiveHeight=1,NegativeHeight=1)
    FalloffSize=0.44
    FlipHeightScaling=False
    FlipWidthScaling=False
    OptionalParentItem=(Type=Bone,Name="calf_r")
    RotationOffset=(X=-90,Y=0,Z=10)
   )
   > &RigUnit_OffsetTransformForItem(
    Item=(Type=Bone,Name="foot_r_deform_base")
    OffsetTransform=(Rotation=(X=0,Y=0.0436,Z=0,W=0.999),Translation=(X=0,Y=0,Z=0),Scale3D=(X=1,Y=1,Z=1))
    bPropagateToChildren=True
    Weight=@n2.OutputParam
   )
   > &RigUnit_OffsetTransformForItem(
    Item=(Type=Bone,Name="foot_r_deform_for")
    OffsetTransform=(Rotation=(X=0,Y=0,Z=0,W=1),Translation=(X=15,Y=0,Z=0),Scale3D=(X=1,Y=1,Z=1))
    bPropagateToChildren=True
    Weight=@n2.OutputParam
   )
   > n3 = &RigUnit_SphericalPoseReader(
    ActiveRegionScaleFactors=(PositiveWidth=1,NegativeWidth=1,PositiveHeight=1,NegativeHeight=1)
    ActiveRegionSize=0.15
    Debug=(bDrawDebug=false,bDraw2D=False,bDrawLocalAxes=False,DebugScale=25,DebugSegments=20,DebugThickness=0.15)
    DriverAxis=(X=-1,Y=0,Z=0)
    DriverItem=(Type=Bone,Name="foot_l")
    FalloffRegionScaleFactors=(PositiveWidth=1,NegativeWidth=1,PositiveHeight=1,NegativeHeight=1)
    FalloffSize=0.5
    FlipHeightScaling=False
    FlipWidthScaling=False
    OptionalParentItem=(Type=Bone,Name="calf_twist_01_l")
    RotationOffset=(X=-90,Y=0,Z=10)
   )
   > &RigUnit_OffsetTransformForItem(
    Item=(Type=Bone,Name="foot_l_deform_base")
    OffsetTransform=(Rotation=(X=0,Y=-0.4226,Z=0,W=0.9063),Translation=(X=0,Y=0,Z=0),Scale3D=(X=1,Y=1,Z=1))
    bPropagateToChildren=True
    Weight=@n3.OutputParam
   )
   > &RigUnit_OffsetTransformForItem(
    Item=(Type=Bone,Name="foot_l_deform_back")
    OffsetTransform=(Rotation=(X=0,Y=0,Z=0,W=1),Translation=(X=11,Y=0,Z=0),Scale3D=(X=1,Y=1,Z=1))
    bPropagateToChildren=True
    Weight=@n3.OutputParam
   )
   > &RigUnit_OffsetTransformForItem(
    Item=(Type=Bone,Name="foot_l_deform_for")
    OffsetTransform=(Rotation=(X=0,Y=0,Z=0,W=1),Translation=(X=-5,Y=0,Z=0),Scale3D=(X=1,Y=1,Z=1))
    bPropagateToChildren=True
    Weight=@n3.OutputParam
   )
   > n4 = &RigUnit_SphericalPoseReader(
    ActiveRegionScaleFactors=(PositiveWidth=1,NegativeWidth=1,PositiveHeight=1,NegativeHeight=1)
    ActiveRegionSize=0.15
    Debug=(bDrawDebug=false,bDraw2D=False,bDrawLocalAxes=False,DebugScale=25,DebugSegments=20,DebugThickness=0.15)
    DriverAxis=(X=-1,Y=0,Z=0)
    DriverItem=(Type=Bone,Name="foot_l")
    FalloffRegionScaleFactors=(PositiveWidth=1,NegativeWidth=1,PositiveHeight=1,NegativeHeight=1)
    FalloffSize=0.44
    FlipHeightScaling=False
    FlipWidthScaling=False
    OptionalParentItem=(Type=Bone,Name="calf_twist_01_l")
    RotationOffset=(X=90,Y=0,Z=10)
   )
   > &RigUnit_OffsetTransformForItem(
    Item=(Type=Bone,Name="foot_l_deform_base")
    OffsetTransform=(Rotation=(X=0,Y=0.0436,Z=0,W=0.999),Translation=(X=0,Y=0,Z=0),Scale3D=(X=1,Y=1,Z=1))
    bPropagateToChildren=True
    Weight=@n4.OutputParam
   )
   > &RigUnit_OffsetTransformForItem(
    Item=(Type=Bone,Name="foot_l_deform_for")
    OffsetTransform=(Rotation=(X=0,Y=0,Z=0,W=1),Translation=(X=15,Y=0,Z=0),Scale3D=(X=1,Y=1,Z=1))
    bPropagateToChildren=True
    Weight=@n4.OutputParam
   )
   > RigVMFunctionReturnNode()
  ]}
 CollapseNode_ContainedGraph {nodes [
   n1 = RigVMFunctionEntryNode()
   &RigUnit_ParentConstraint(
    AdvancedSettings=(InterpolationType=Average,RotationOrderForFilter=XZY)
    Child=(Type=Bone,Name="belt_for_r")
    Parents="((Item=(Type=Bone,Name=\"pelvis\"),Weight=0.0433),(Item=(Type=Bone,Name=\"spine_01\"),Weight=0.1103),(Item=(Type=Bone,Name=\"spine_02\"),Weight=0.255),(Item=(Type=Bone,Name=\"spine_03\"),Weight=0.3844),(Item=(Type=Bone,Name=\"clavicle_r\"),Weight=0.056),(Item=(Type= ...(truncated)"
    Weight=1
    ExecutePin=@n1.ExecutePin
   )
   > &RigUnit_ParentConstraint(
    AdvancedSettings=(InterpolationType=Average,RotationOrderForFilter=XZY)
    Child=(Type=Bone,Name="belt_for_l")
    Parents="((Item=(Type=Bone,Name=\"pelvis\"),Weight=0.0433),(Item=(Type=Bone,Name=\"spine_01\"),Weight=0.1103),(Item=(Type=Bone,Name=\"spine_02\"),Weight=0.255),(Item=(Type=Bone,Name=\"spine_03\"),Weight=0.3844),(Item=(Type=Bone,Name=\"clavicle_l\"),Weight=0.056),(Item=(Type= ...(truncated)"
    Weight=1
   )
   > &RigUnit_ParentConstraint(
    AdvancedSettings=(InterpolationType=Average,RotationOrderForFilter=XZY)
    Child=(Type=Bone,Name="belt_for_up_r")
    Parents="((Item=(Type=Bone,Name=\"pelvis\"),Weight=0.0192),(Item=(Type=Bone,Name=\"spine_01\"),Weight=0.0788),(Item=(Type=Bone,Name=\"spine_02\"),Weight=0.2018),(Item=(Type=Bone,Name=\"spine_03\"),Weight=0.3641),(Item=(Type=Bone,Name=\"spine_04\"),Weight=0.1703),(Item=(Type= ...(truncated)"
    Weight=1
   )
   > &RigUnit_ParentConstraint(
    AdvancedSettings=(InterpolationType=Average,RotationOrderForFilter=XZY)
    Child=(Type=Bone,Name="belt_for_up_l")
    Parents="((Item=(Type=Bone,Name=\"pelvis\"),Weight=0.0192),(Item=(Type=Bone,Name=\"spine_01\"),Weight=0.0788),(Item=(Type=Bone,Name=\"spine_02\"),Weight=0.2018),(Item=(Type=Bone,Name=\"spine_03\"),Weight=0.3641),(Item=(Type=Bone,Name=\"spine_04\"),Weight=0.1703),(Item=(Type= ...(truncated)"
    Weight=1
   )
   > n2 = &RigUnit_SphericalPoseReader(
    ActiveRegionScaleFactors=(PositiveWidth=1,NegativeWidth=1,PositiveHeight=1,NegativeHeight=1)
    ActiveRegionSize=0.2
    Debug=(bDrawDebug=false,bDraw2D=False,bDrawLocalAxes=False,DebugScale=25,DebugSegments=20,DebugThickness=0.25)
    DriverAxis=(X=0,Y=1,Z=0)
    DriverItem=(Type=Bone,Name="spine_05")
    FalloffRegionScaleFactors=(PositiveWidth=1,NegativeWidth=1,PositiveHeight=1,NegativeHeight=1)
    FalloffSize=0.5
    FlipHeightScaling=False
    FlipWidthScaling=False
    OptionalParentItem=(Type=Bone,Name="spine_03")
    RotationOffset=(X=180,Y=0,Z=0)
   )
   > RigUnit_OffsetTransformForItem = &RigUnit_OffsetTransformForItem(
    Item=(Type=Bone,Name="belt_for_l")
    OffsetTransform=(Rotation=(X=-0.2588,Y=0,Z=0,W=0.9659),Translation=(X=0,Y=-1.5,Z=0),Scale3D=(X=1,Y=1,Z=1))
    bPropagateToChildren=True
    Weight=@n2.OutputParam
   )
   > &RigUnit_OffsetTransformForItem(
    Item=(Type=Bone,Name="belt_for_r")
    OffsetTransform=(Rotation=(X=0.2588,Y=0,Z=0,W=0.9659),Translation=(X=0,Y=-1.5,Z=0),Scale3D=(X=1,Y=1,Z=1))
    bPropagateToChildren=True
    Weight=@n2.OutputParam
   )
   > &RigUnit_OffsetTransformForItem(
    Item=(Type=Bone,Name="belt_for_up_l")
    OffsetTransform=(Rotation=(X=0.2588,Y=0,Z=0,W=0.9659),Translation=(X=0,Y=-1.5,Z=0),Scale3D=(X=1,Y=1,Z=1))
    bPropagateToChildren=True
    Weight=@n2.OutputParam
   )
   > &RigUnit_OffsetTransformForItem(
    Item=(Type=Bone,Name="belt_for_up_r")
    OffsetTransform=(Rotation=(X=0.1305,Y=0,Z=0,W=0.9914),Translation=(X=0,Y=-1,Z=0),Scale3D=(X=1,Y=1,Z=1))
    bPropagateToChildren=True
    Weight=@n2.OutputParam
   )
   > n3 = &RigUnit_SphericalPoseReader(
    ActiveRegionScaleFactors=(PositiveWidth=1,NegativeWidth=1,PositiveHeight=1,NegativeHeight=1)
    ActiveRegionSize=0.2
    Debug=(bDrawDebug=false,bDraw2D=False,bDrawLocalAxes=False,DebugScale=25,DebugSegments=20,DebugThickness=0.25)
    DriverAxis=(X=0,Y=1,Z=0)
    DriverItem=(Type=Bone,Name="spine_05")
    FalloffRegionScaleFactors=(PositiveWidth=1,NegativeWidth=1,PositiveHeight=1,NegativeHeight=1)
    FalloffSize=0.5
    FlipHeightScaling=False
    FlipWidthScaling=False
    OptionalParentItem=(Type=Bone,Name="spine_02")
    RotationOffset=(X=0,Y=0,Z=0)
   )
   > &RigUnit_OffsetTransformForItem(
    Item=(Type=Bone,Name="belt_for_l")
    OffsetTransform=(Rotation=(X=0.1305,Y=0,Z=0,W=0.9914),Translation=(X=0,Y=1.5,Z=0),Scale3D=(X=1,Y=1,Z=1))
    bPropagateToChildren=True
    Weight=@n3.OutputParam
   )
   > n4 = &RigUnit_OffsetTransformForItem(
    Item=(Type=Bone,Name="belt_for_r")
    OffsetTransform=(Rotation=(X=-0.4226,Y=0,Z=0,W=0.9063),Translation=(X=0,Y=2.5,Z=0),Scale3D=(X=1,Y=1,Z=1))
    bPropagateToChildren=True
    Weight=@n3.OutputParam
   )
   RigVMFunctionReturnNode(ExecutePin=@n4.ExecutePin)
  ]}
 CollapseNode_ContainedGraph {nodes [
   n1 = RigVMFunctionEntryNode()
   &RigUnit_ParentConstraint(
    AdvancedSettings=(InterpolationType=Average,RotationOrderForFilter=XZY)
    Child=(Type=Bone,Name="pcoket_A_r")
    Parents="((Item=(Type=Bone,Name=\"pelvis\"),Weight=0.8193),(Item=(Type=Bone,Name=\"spine_01\"),Weight=0.0516),(Item=(Type=Bone,Name=\"spine_02\"),Weight=0.0601),(Item=(Type=Bone,Name=\"spine_03\"),Weight=0.0329),(Item=(Type=Bone,Name=\"thigh_twist_01_r\"),Weight=0.025),(Item ...(truncated)"
    Weight=1
    ExecutePin=@n1.ExecutePin
   )
   > &RigUnit_ParentConstraint(
    AdvancedSettings=(InterpolationType=Average,RotationOrderForFilter=XZY)
    Child=(Type=Bone,Name="pcoket_B_r")
    Parents="((Item=(Type=Bone,Name=\"pelvis\"),Weight=0.7836),(Item=(Type=Bone,Name=\"spine_01\"),Weight=0.0665),(Item=(Type=Bone,Name=\"spine_02\"),Weight=0.077),(Item=(Type=Bone,Name=\"spine_03\"),Weight=0.0441),(Item=(Type=Bone,Name=\"thigh_twist_01_r\"),Weight=0.0128),(Item ...(truncated)"
    Weight=1
   )
   > &RigUnit_ParentConstraint(
    AdvancedSettings=(InterpolationType=Average,RotationOrderForFilter=XZY)
    Child=(Type=Bone,Name="pocket_mid")
    Parents="((Item=(Type=Bone,Name=\"pelvis\"),Weight=0.8286),(Item=(Type=Bone,Name=\"spine_01\"),Weight=0.0504),(Item=(Type=Bone,Name=\"spine_02\"),Weight=0.0586),(Item=(Type=Bone,Name=\"spine_03\"),Weight=0.0345),(Item=(Type=Bone,Name=\"thigh_twist_01_l\"),Weight=0.0085),(Ite ...(truncated)"
    Weight=1
   )
   > &RigUnit_ParentConstraint(
    AdvancedSettings=(InterpolationType=Average,RotationOrderForFilter=XZY)
    Child=(Type=Bone,Name="pcoket_B_l")
    Parents="((Item=(Type=Bone,Name=\"pelvis\"),Weight=0.7835),(Item=(Type=Bone,Name=\"spine_01\"),Weight=0.0665),(Item=(Type=Bone,Name=\"spine_02\"),Weight=0.077),(Item=(Type=Bone,Name=\"spine_03\"),Weight=0.0441),(Item=(Type=Bone,Name=\"thigh_twist_01_l\"),Weight=0.0128),(Item ...(truncated)"
    Weight=1
   )
   > n2 = &RigUnit_ParentConstraint(
    AdvancedSettings=(InterpolationType=Average,RotationOrderForFilter=XZY)
    Child=(Type=Bone,Name="pcoket_A_l")
    Parents="((Item=(Type=Bone,Name=\"pelvis\"),Weight=0.8298),(Item=(Type=Bone,Name=\"spine_01\"),Weight=0.0504),(Item=(Type=Bone,Name=\"spine_02\"),Weight=0.0588),(Item=(Type=Bone,Name=\"spine_03\"),Weight=0.0322),(Item=(Type=Bone,Name=\"thigh_twist_01_l\"),Weight=0.0224),(Ite ...(truncated)"
    Weight=1
   )
   RigVMFunctionReturnNode(ExecutePin=@n2.ExecutePin)
  ]}
 CollapseNode_ContainedGraph {nodes [
   n1 = RigVMFunctionEntryNode()
   &RigUnit_ParentConstraint(
    AdvancedSettings=(InterpolationType=Average,RotationOrderForFilter=XZY)
    Child=(Type=Bone,Name="belt_chain_0")
    Parents="((Item=(Type=Bone,Name=\"spine_02\"),Weight=0.0917),(Item=(Type=Bone,Name=\"spine_03\"),Weight=0.2106),(Item=(Type=Bone,Name=\"spine_04\"),Weight=0.4801),(Item=(Type=Bone,Name=\"spine_05\"),Weight=0.2011),(Item=(Type=Bone,Name=\"clavicle_l\"),Weight=0.0067),(Item=(T ...(truncated)"
    Weight=1
    ExecutePin=@n1.ExecutePin
   )
   > &RigUnit_ParentConstraint(
    AdvancedSettings=(InterpolationType=Average,RotationOrderForFilter=XZY)
    Child=(Type=Bone,Name="belt_chain_1")
    Parents="((Item=(Type=Bone,Name=\"spine_02\"),Weight=0.0128),(Item=(Type=Bone,Name=\"spine_03\"),Weight=0.0513),(Item=(Type=Bone,Name=\"spine_04\"),Weight=0.1569),(Item=(Type=Bone,Name=\"spine_05\"),Weight=0.4763),(Item=(Type=Bone,Name=\"neck_01\"),Weight=0.0071),(Item=(Type ...(truncated)"
    Weight=1
   )
   > &RigUnit_ParentConstraint(
    AdvancedSettings=(InterpolationType=Average,RotationOrderForFilter=XZY)
    Child=(Type=Bone,Name="belt_chain_2")
    Parents="((Item=(Type=Bone,Name=\"spine_03\"),Weight=0.0085),(Item=(Type=Bone,Name=\"spine_04\"),Weight=0.0513),(Item=(Type=Bone,Name=\"spine_05\"),Weight=0.6552),(Item=(Type=Bone,Name=\"neck_01\"),Weight=0.024),(Item=(Type=Bone,Name=\"neck_02\"),Weight=0.0003),(Item=(Type=B ...(truncated)"
    Weight=1
   )
   > &RigUnit_ParentConstraint(
    AdvancedSettings=(InterpolationType=Average,RotationOrderForFilter=XZY)
    Child=(Type=Bone,Name="belt_chain_3")
    Parents="((Item=(Type=Bone,Name=\"spine_04\"),Weight=0.0008),(Item=(Type=Bone,Name=\"spine_05\"),Weight=0.6048),(Item=(Type=Bone,Name=\"neck_01\"),Weight=0.0253),(Item=(Type=Bone,Name=\"clavicle_l\"),Weight=0.3476),(Item=(Type=Bone,Name=\"upperarm_twist_01_l\"),Weight=0.0087 ...(truncated)"
    Weight=1
   )
   > &RigUnit_ParentConstraint(
    AdvancedSettings=(InterpolationType=Average,RotationOrderForFilter=XZY)
    Child=(Type=Bone,Name="belt_chain_4")
    Parents="((Item=(Type=Bone,Name=\"spine_03\"),Weight=0.0081),(Item=(Type=Bone,Name=\"spine_04\"),Weight=0.0017),(Item=(Type=Bone,Name=\"spine_05\"),Weight=0.6864),(Item=(Type=Bone,Name=\"neck_01\"),Weight=0.0198),(Item=(Type=Bone,Name=\"clavicle_l\"),Weight=0.2782),(Item=(Ty ...(truncated)"
    Weight=1
   )
   > &RigUnit_ParentConstraint(
    AdvancedSettings=(InterpolationType=Average,RotationOrderForFilter=XZY)
    Child=(Type=Bone,Name="belt_chain_5")
    Parents="((Item=(Type=Bone,Name=\"spine_03\"),Weight=0.1367),(Item=(Type=Bone,Name=\"spine_04\"),Weight=0.0364),(Item=(Type=Bone,Name=\"spine_05\"),Weight=0.5421),(Item=(Type=Bone,Name=\"clavicle_l\"),Weight=0.2707),(Item=(Type=Bone,Name=\"upperarm_twist_01_l\"),Weight=0.008 ...(truncated)"
    Weight=1
   )
   > n2 = &RigUnit_SphericalPoseReader(
    ActiveRegionScaleFactors=(PositiveWidth=1,NegativeWidth=1,PositiveHeight=1,NegativeHeight=1)
    ActiveRegionSize=0.2
    Debug=(bDrawDebug=false,bDraw2D=False,bDrawLocalAxes=False,DebugScale=25,DebugSegments=20,DebugThickness=0.25)
    DriverAxis=(X=0,Y=1,Z=0)
    DriverItem=(Type=Bone,Name="spine_05")
    FalloffRegionScaleFactors=(PositiveWidth=1,NegativeWidth=1,PositiveHeight=1,NegativeHeight=1)
    FalloffSize=0.5
    FlipHeightScaling=False
    FlipWidthScaling=False
    OptionalParentItem=(Type=Bone,Name="spine_03")
    RotationOffset=(X=180,Y=0,Z=0)
   )
   > RigUnit_OffsetTransformForItem = &RigUnit_OffsetTransformForItem(
    Item=(Type=Bone,Name="belt_chain_0")
    OffsetTransform=(Rotation=(X=0.1736,Y=0,Z=0,W=0.9848),Translation=(X=0.7,Y=0,Z=-1.5),Scale3D=(X=1,Y=1,Z=1))
    bPropagateToChildren=True
    Weight=@n2.OutputParam
   )
   > n3 = &RigUnit_SphericalPoseReader(
    ActiveRegionScaleFactors=(PositiveWidth=1,NegativeWidth=1,PositiveHeight=1,NegativeHeight=1)
    ActiveRegionSize=0.2
    Debug=(bDrawDebug=false,bDraw2D=False,bDrawLocalAxes=False,DebugScale=25,DebugSegments=20,DebugThickness=0.25)
    DriverAxis=(X=0,Y=1,Z=0)
    DriverItem=(Type=Bone,Name="spine_05")
    FalloffRegionScaleFactors=(PositiveWidth=1,NegativeWidth=1,PositiveHeight=1,NegativeHeight=1)
    FalloffSize=0.5
    FlipHeightScaling=False
    FlipWidthScaling=False
    OptionalParentItem=(Type=Bone,Name="spine_02")
    RotationOffset=(X=0,Y=0,Z=0)
   )
   > &RigUnit_OffsetTransformForItem(
    Item=(Type=Bone,Name="belt_chain_0")
    OffsetTransform=(Rotation=(X=-0.1736,Y=0,Z=0,W=0.9848),Translation=(X=0.7,Y=0,Z=1.5),Scale3D=(X=1,Y=1,Z=1))
    bPropagateToChildren=True
    Weight=@n3.OutputParam
   )
   > n4 = &RigUnit_SphericalPoseReader(
    ActiveRegionScaleFactors=(PositiveWidth=1,NegativeWidth=1,PositiveHeight=1,NegativeHeight=1)
    ActiveRegionSize=0.1
    Debug=(bDrawDebug=false,bDraw2D=False,bDrawLocalAxes=False,DebugScale=15,DebugSegments=20,DebugThickness=0.1)
    DriverAxis=(X=1,Y=0,Z=0)
    DriverItem=(Type=Bone,Name="upperarm_l")
    FalloffRegionScaleFactors=(PositiveWidth=1,NegativeWidth=1,PositiveHeight=1,NegativeHeight=1)
    FalloffSize=0.4
    FlipHeightScaling=False
    FlipWidthScaling=False
    OptionalParentItem=(Type=Bone,Name="spine_05")
    RotationOffset=(X=0,Y=0,Z=0)
   )
   > &RigUnit_OffsetTransformForItem(
    Item=(Type=Bone,Name="belt_chain_3")
    OffsetTransform=(Rotation=(X=0,Y=0,Z=0,W=1),Translation=(X=1,Y=0,Z=0),Scale3D=(X=1,Y=1,Z=1))
    bPropagateToChildren=True
    Weight=@n4.OutputParam
   )
   > &RigUnit_OffsetTransformForItem(
    Item=(Type=Bone,Name="belt_chain_2")
    OffsetTransform=(Rotation=(X=0,Y=-0.0872,Z=0,W=0.9962),Translation=(X=1.2,Y=0,Z=0),Scale3D=(X=1,Y=1,Z=1))
    bPropagateToChildren=True
    Weight=&RigVMFunction_KalmanFloat(BufferSize=10, Value=@n4.OutputParam) {template="KalmanFilter::Execute(in Value,in BufferSize,out Result)"}.Result
   )
   > &RigUnit_OffsetTransformForItem(
    Item=(Type=Bone,Name="belt_chain_1")
    OffsetTransform=(Rotation=(X=0,Y=0,Z=0,W=1),Translation=(X=0.5,Y=0,Z=0.5),Scale3D=(X=1,Y=1,Z=1))
    bPropagateToChildren=True
    Weight=@n4.OutputParam
   )
   > &RigUnit_OffsetTransformForItem(
    Item=(Type=Bone,Name="belt_chain_0")
    OffsetTransform=(Rotation=(X=0,Y=0,Z=0,W=1),Translation=(X=0,Y=0,Z=0.7),Scale3D=(X=1,Y=1,Z=1))
    bPropagateToChildren=True
    Weight=@n4.OutputParam
   )
   > &RigUnit_OffsetTransformForItem(
    Item=(Type=Bone,Name="belt_chain_4")
    OffsetTransform=(Rotation=(X=0.0436,Y=0,Z=0,W=0.999),Translation=(X=0,Y=0,Z=1.5),Scale3D=(X=1,Y=1,Z=1))
    bPropagateToChildren=True
    Weight=@n4.OutputParam
   )
   > n5 = &RigUnit_SphericalPoseReader(
    ActiveRegionScaleFactors=(PositiveWidth=1.5,NegativeWidth=1,PositiveHeight=1,NegativeHeight=1)
    ActiveRegionSize=0.1
    Debug=(bDrawDebug=false,bDraw2D=False,bDrawLocalAxes=False,DebugScale=15,DebugSegments=20,DebugThickness=0.1)
    DriverAxis=(X=1,Y=0,Z=0)
    DriverItem=(Type=Bone,Name="upperarm_l")
    FalloffRegionScaleFactors=(PositiveWidth=2.5,NegativeWidth=1,PositiveHeight=1,NegativeHeight=1)
    FalloffSize=0.2
    FlipHeightScaling=False
    FlipWidthScaling=False
    OptionalParentItem=(Type=Bone,Name="spine_05")
    RotationOffset=(X=15,Y=45,Z=0)
   )
   > &RigUnit_OffsetTransformForItem(
    Item=(Type=Bone,Name="belt_chain_1")
    OffsetTransform=(Rotation=(X=0,Y=-0.1736,Z=0,W=0.9848),Translation=(X=0,Y=0,Z=1),Scale3D=(X=1,Y=1,Z=1))
    bPropagateToChildren=True
    Weight=@n5.OutputParam
   )
   KalmanFilter = &RigVMFunction_KalmanFloat(BufferSize=12, Value=@n5.OutputParam) {template="KalmanFilter::Execute(in Value,in BufferSize,out Result)"}
   > &RigUnit_OffsetTransformForItem(
    Item=(Type=Bone,Name="belt_chain_0")
    OffsetTransform=(Rotation=(X=-0.0523,Y=0,Z=0,W=0.9986),Translation=(X=0,Y=0,Z=1),Scale3D=(X=1,Y=1,Z=1))
    bPropagateToChildren=True
    Weight=@KalmanFilter.Result
   )
   > &RigUnit_OffsetTransformForItem(
    Item=(Type=Bone,Name="belt_chain_2")
    OffsetTransform=(Rotation=(X=0,Y=-0.5,Z=0,W=0.866),Translation=(X=1.5,Y=0,Z=0),Scale3D=(X=1,Y=1,Z=1))
    bPropagateToChildren=True
    Weight=@n5.OutputParam
   )
   > &RigUnit_OffsetTransformForItem(
    Item=(Type=Bone,Name="belt_chain_3")
    OffsetTransform=(Rotation=(X=0,Y=-0.3007,Z=0,W=0.9537),Translation=(X=0,Y=0,Z=0),Scale3D=(X=1,Y=1,Z=1))
    bPropagateToChildren=True
    Weight=@n5.OutputParam
   )
   > n6 = &RigUnit_SphericalPoseReader(
    ActiveRegionScaleFactors=(PositiveWidth=1,NegativeWidth=1,PositiveHeight=1,NegativeHeight=1)
    ActiveRegionSize=0.1
    Debug=(bDrawDebug=false,bDraw2D=False,bDrawLocalAxes=False,DebugScale=15,DebugSegments=20,DebugThickness=0.1)
    DriverAxis=(X=1,Y=0,Z=0)
    DriverItem=(Type=Bone,Name="upperarm_l")
    FalloffRegionScaleFactors=(PositiveWidth=1,NegativeWidth=1,PositiveHeight=1,NegativeHeight=1)
    FalloffSize=0.2
    FlipHeightScaling=False
    FlipWidthScaling=False
    OptionalParentItem=(Type=Bone,Name="spine_05")
    RotationOffset=(X=45,Y=15,Z=20)
   )
   > &RigUnit_OffsetTransformForItem(
    Item=(Type=Bone,Name="belt_chain_shoulder")
    OffsetTransform=(Rotation=(X=0.0273,Y=-0.2149,Z=0.087,W=0.9724),Translation=(X=0,Y=1,Z=0.5),Scale3D=(X=1,Y=1.3,Z=1))
    bPropagateToChildren=True
    Weight=@n6.OutputParam
   )
   > n7 = &RigUnit_SphericalPoseReader(
    ActiveRegionScaleFactors=(PositiveWidth=1.1,NegativeWidth=1,PositiveHeight=1,NegativeHeight=1.7)
    ActiveRegionSize=0.07
    Debug=(bDrawDebug=true,bDraw2D=False,bDrawLocalAxes=False,DebugScale=15,DebugSegments=20,DebugThickness=0.1)
    DriverAxis=(X=1,Y=0,Z=0)
    DriverItem=(Type=Bone,Name="upperarm_l")
    FalloffRegionScaleFactors=(PositiveWidth=1.4,NegativeWidth=1,PositiveHeight=1,NegativeHeight=3)
    FalloffSize=0.15
    FlipHeightScaling=False
    FlipWidthScaling=False
    OptionalParentItem=(Type=Bone,Name="spine_05")
    RotationOffset=(X=60,Y=0,Z=0)
   )
   > &RigUnit_OffsetTransformForItem(
    Item=(Type=Bone,Name="belt_chain_1")
    OffsetTransform=(Rotation=(X=0,Y=0,Z=0,W=1),Translation=(X=1.2,Y=0,Z=0),Scale3D=(X=1,Y=1,Z=1))
    bPropagateToChildren=True
    Weight=@n7.OutputParam
   )
   > &RigUnit_OffsetTransformForItem(
    Item=(Type=Bone,Name="belt_chain_2")
    OffsetTransform=(Rotation=(X=0,Y=-0.061,Z=0,W=0.9981),Translation=(X=1.5,Y=0,Z=0),Scale3D=(X=1,Y=1,Z=1))
    bPropagateToChildren=True
    Weight=@n7.OutputParam
   )
   > &RigUnit_OffsetTransformForItem(
    Item=(Type=Bone,Name="belt_chain_4")
    OffsetTransform=(Rotation=(X=0,Y=0,Z=0,W=1),Translation=(X=0.5,Y=0,Z=0),Scale3D=(X=1,Y=1,Z=1))
    bPropagateToChildren=True
    Weight=@n7.OutputParam
   )
   > n8 = &RigUnit_SphericalPoseReader(
    ActiveRegionScaleFactors=(PositiveWidth=1,NegativeWidth=1,PositiveHeight=1,NegativeHeight=1)
    ActiveRegionSize=0.07
    Debug=(bDrawDebug=false,bDraw2D=False,bDrawLocalAxes=False,DebugScale=15,DebugSegments=20,DebugThickness=0.1)
    DriverAxis=(X=1,Y=0,Z=0)
    DriverItem=(Type=Bone,Name="upperarm_l")
    FalloffRegionScaleFactors=(PositiveWidth=1.5,NegativeWidth=1,PositiveHeight=1.5,NegativeHeight=1)
    FalloffSize=0.3
    FlipHeightScaling=False
    FlipWidthScaling=False
    OptionalParentItem=(Type=Bone,Name="spine_05")
    RotationOffset=(X=60,Y=0,Z=0)
   )
   > &RigUnit_OffsetTransformForItem(
    Item=(Type=Bone,Name="deform_pec_belt")
    OffsetTransform=(Rotation=(X=0,Y=0,Z=0,W=1),Translation=(X=0,Y=1.5,Z=-3),Scale3D=(X=1,Y=1,Z=1))
    bPropagateToChildren=True
    Weight=@n8.OutputParam
   )
   > n9 = &RigUnit_SphericalPoseReader(
    ActiveRegionScaleFactors=(PositiveWidth=1,NegativeWidth=1,PositiveHeight=1,NegativeHeight=1)
    ActiveRegionSize=0.07
    Debug=(bDrawDebug=false,bDraw2D=False,bDrawLocalAxes=False,DebugScale=15,DebugSegments=20,DebugThickness=0.1)
    DriverAxis=(X=1,Y=0,Z=0)
    DriverItem=(Type=Bone,Name="upperarm_l")
    FalloffRegionScaleFactors=(PositiveWidth=1,NegativeWidth=1,PositiveHeight=1,NegativeHeight=2)
    FalloffSize=0.15
    FlipHeightScaling=False
    FlipWidthScaling=False
    OptionalParentItem=(Type=Bone,Name="spine_05")
    RotationOffset=(X=120,Y=-80,Z=0)
   )
   > &RigUnit_OffsetTransformForItem(
    Item=(Type=Bone,Name="belt_chain_2")
    OffsetTransform=(Rotation=(X=0,Y=-0.0872,Z=0,W=0.9962),Translation=(X=0.7,Y=0,Z=0),Scale3D=(X=1,Y=1,Z=1))
    bPropagateToChildren=True
    Weight=@n9.OutputParam
   )
   > &RigUnit_OffsetTransformForItem(
    Item=(Type=Bone,Name="belt_chain_1")
    OffsetTransform=(Rotation=(X=0,Y=0,Z=0,W=1),Translation=(X=0.7,Y=0,Z=0),Scale3D=(X=1,Y=1,Z=1))
    bPropagateToChildren=True
    Weight=@n9.OutputParam
   )
   > n10 = &RigUnit_SphericalPoseReader(
    ActiveRegionScaleFactors=(PositiveWidth=1,NegativeWidth=1,PositiveHeight=1,NegativeHeight=1)
    ActiveRegionSize=0.05
    Debug=(bDrawDebug=false,bDraw2D=False,bDrawLocalAxes=False,DebugScale=15,DebugSegments=20,DebugThickness=0.1)
    DriverAxis=(X=1,Y=0,Z=0)
    DriverItem=(Type=Bone,Name="upperarm_l")
    FalloffRegionScaleFactors=(PositiveWidth=1,NegativeWidth=1,PositiveHeight=1,NegativeHeight=1)
    FalloffSize=0.17
    FlipHeightScaling=False
    FlipWidthScaling=False
    OptionalParentItem=(Type=Bone,Name="spine_05")
    RotationOffset=(X=45,Y=40,Z=0)
   )
   > &RigUnit_OffsetTransformForItem(
    Item=(Type=Bone,Name="belt_chain_2")
    OffsetTransform=(Rotation=(X=0,Y=-0.342,Z=0,W=0.9397),Translation=(X=0.6,Y=0,Z=0),Scale3D=(X=1,Y=1,Z=1))
    bPropagateToChildren=True
    Weight=@n10.OutputParam
   )
   > &RigUnit_OffsetTransformForItem(
    Item=(Type=Bone,Name="belt_chain_3")
    OffsetTransform=(Rotation=(X=0.0655,Y=-0.3412,Z=0.0239,W=0.9374),Translation=(X=0.8,Y=0,Z=0),Scale3D=(X=1,Y=1,Z=1))
    bPropagateToChildren=True
    Weight=@n10.OutputParam
   )
   > &RigUnit_OffsetTransformForItem(
    Item=(Type=Bone,Name="belt_chain_4")
    OffsetTransform=(Rotation=(X=0,Y=0,Z=0,W=1),Translation=(X=0.5,Y=0,Z=0),Scale3D=(X=1,Y=1,Z=1))
    bPropagateToChildren=True
    Weight=@n10.OutputParam
   )
   RigUnit_GetTransform = &RigUnit_GetTransform(Item=(Type=Bone,Name="buckle_pnt"), bInitial=False)
   RigVMFunction_MathVectorDistance = &RigVMFunction_MathVectorDistance(
    A=@RigUnit_GetTransform.Transform.Translation
    B=&RigUnit_GetTransform(Item=(Type=Bone,Name="belt_chain_fr_base"), bInitial=False).Transform.Translation
   )
   Remap = &RigVMFunction_MathDoubleRemap(SourceMaximum=2, SourceMinimum=4.8, TargetMaximum=1, TargetMinimum=0, Value=@RigVMFunction_MathVectorDistance.Result)
   > &RigUnit_OffsetTransformForItem(
    Item=(Type=Bone,Name="belt_chain_1")
    OffsetTransform=(Rotation=(X=0,Y=-0.2334,Z=0,W=0.9724),Translation=(X=0,Y=0,Z=0.4),Scale3D=(X=1,Y=1,Z=1))
    bPropagateToChildren=True
    Weight=@Remap.Result
   )
   > &RigUnit_OffsetTransformForItem(
    Item=(Type=Bone,Name="belt_chain_0")
    OffsetTransform=(Rotation=(X=-0.0851,Y=-0.2156,Z=-0.0189,W=0.9726),Translation=(X=0.5,Y=0,Z=0.5),Scale3D=(X=1,Y=1,Z=1))
    bPropagateToChildren=True
    Weight=@Remap.Result
   )
   n11 = &RigUnit_GetTransform(Item=(Type=Bone,Name="belt_chain_shoulder_pnt"), bInitial=False)
   n12 = &RigUnit_GetTransform(Item=(Type=Bone,Name="shoulder_l_pnt"), bInitial=False)
   > n13 = &RigUnit_OffsetTransformForItem(
    Item=(Type=Bone,Name="belt_chain_shoulder")
    OffsetTransform=(Rotation=(X=-0.0449,Y=-0.2549,Z=-0.1677,W=0.9513),Translation=(X=1,Y=2,Z=1),Scale3D=(X=1,Y=1.4,Z=1))
    bPropagateToChildren=True
    Weight=&RigVMFunction_MathDoubleRemap(
     SourceMaximum=3
     SourceMinimum=5
     TargetMaximum=1
     TargetMinimum=0
     Value=&RigVMFunction_MathVectorDistance(A=@n11.Transform.Translation, B=@n12.Transform.Translation).Result
    ).Result
   )
   RigVMFunctionReturnNode(ExecutePin=@n13.ExecutePin)
  ]}
 CollapseNode_ContainedGraph {nodes [
   n1 = RigVMFunctionEntryNode()
   n2 = &RigUnit_SphericalPoseReader(
    ActiveRegionScaleFactors=(PositiveWidth=1,NegativeWidth=1,PositiveHeight=1,NegativeHeight=1)
    ActiveRegionSize=0.1
    Debug=(bDrawDebug=true,bDraw2D=False,bDrawLocalAxes=False,DebugScale=15,DebugSegments=20,DebugThickness=0.1)
    DriverAxis=(X=-1,Y=0,Z=0)
    DriverItem=(Type=Bone,Name="thigh_l")
    FalloffRegionScaleFactors=(PositiveWidth=1,NegativeWidth=1,PositiveHeight=1,NegativeHeight=1)
    FalloffSize=0.33
    FlipHeightScaling=False
    FlipWidthScaling=False
    OptionalParentItem=(Type=Bone,Name="None")
    RotationOffset=(X=125,Y=35,Z=0)
    ExecutePin=@n1.ExecutePin
   )
   > RigUnit_OffsetTransformForItem = &RigUnit_OffsetTransformForItem(
    Item=(Type=Bone,Name="pcoket_A_l")
    OffsetTransform=(Rotation=(X=0.0046,Y=0.0346,Z=-0.1304,W=0.9908),Translation=(X=0,Y=-1.5,Z=1.5),Scale3D=(X=1,Y=1,Z=1))
    bPropagateToChildren=True
    Weight=@n2.OutputParam
   )
   > &RigUnit_OffsetTransformForItem(
    Item=(Type=Bone,Name="pcoket_B_l")
    OffsetTransform=(Rotation=(X=-0.0283,Y=0.2146,Z=0.1274,W=0.9679),Translation=(X=0,Y=2.5,Z=2.5),Scale3D=(X=1,Y=1,Z=1))
    bPropagateToChildren=True
    Weight=@n2.OutputParam
   )
   Minimum = /Script/RigVM.RigVMFunction_MathDoubleMin(B=RigVMVariableNode() {variable=is moving}.Value, A=@n2.OutputParam) {is=struct RigVMUnitNode, template="Minimum::Execute(in A,in B,out Result)"}
   > &RigUnit_OffsetTransformForItem(
    Item=(Type=Bone,Name="pcoket_B_l")
    OffsetTransform=(Rotation=(X=0,Y=0,Z=0,W=1),Translation=(X=0,Y=-2,Z=-1.5),Scale3D=(X=1,Y=1,Z=1))
    bPropagateToChildren=True
    Weight=@Minimum.Result
   )
   > &RigUnit_OffsetTransformForItem(
    Item=(Type=Bone,Name="pocket_mid")
    OffsetTransform=(Rotation=(X=-0.13,Y=0.0864,Z=0.0114,W=0.9877),Translation=(X=1,Y=1,Z=2),Scale3D=(X=1,Y=1,Z=1))
    bPropagateToChildren=True
    Weight=@n2.OutputParam
   )
   > n3 = &RigUnit_SphericalPoseReader(
    ActiveRegionScaleFactors=(PositiveWidth=1,NegativeWidth=1,PositiveHeight=1,NegativeHeight=1)
    ActiveRegionSize=0.15
    Debug=(bDrawDebug=false,bDraw2D=False,bDrawLocalAxes=False,DebugScale=15,DebugSegments=20,DebugThickness=0.1)
    DriverAxis=(X=1,Y=0,Z=0)
    DriverItem=(Type=Bone,Name="thigh_r")
    FalloffRegionScaleFactors=(PositiveWidth=1,NegativeWidth=1,PositiveHeight=1,NegativeHeight=1)
    FalloffSize=0.33
    FlipHeightScaling=False
    FlipWidthScaling=False
    OptionalParentItem=(Type=Bone,Name="None")
    RotationOffset=(X=-90,Y=0,Z=-15)
   )
   > &RigUnit_OffsetTransformForItem(
    Item=(Type=Bone,Name="pcoket_B_r")
    OffsetTransform=(Rotation=(X=0.0121,Y=0.1732,Z=-0.0687,W=0.9824),Translation=(X=0.7,Y=0,Z=2.5),Scale3D=(X=1,Y=1,Z=1))
    bPropagateToChildren=True
    Weight=@n3.OutputParam
   )
   > &RigUnit_OffsetTransformForItem(
    Item=(Type=Bone,Name="pocket_mid")
    OffsetTransform=(Rotation=(X=0.1304,Y=0.0432,Z=-0.0057,W=0.9905),Translation=(X=1.5,Y=0,Z=2),Scale3D=(X=1,Y=1,Z=1))
    bPropagateToChildren=True
    Weight=@n3.OutputParam
   )
   > &RigUnit_OffsetTransformForItem(
    Item=(Type=Bone,Name="holster")
    OffsetTransform=(Rotation=(X=-0.0872,Y=0,Z=0,W=0.9962),Translation=(X=0,Y=1,Z=0),Scale3D=(X=1,Y=1,Z=1))
    bPropagateToChildren=True
    Weight=@n3.OutputParam
   )
   > n4 = &RigUnit_SphericalPoseReader(
    ActiveRegionScaleFactors=(PositiveWidth=1,NegativeWidth=1,PositiveHeight=1,NegativeHeight=1)
    ActiveRegionSize=0.1
    Debug=(bDrawDebug=false,bDraw2D=False,bDrawLocalAxes=False,DebugScale=15,DebugSegments=20,DebugThickness=0.1)
    DriverAxis=(X=-1,Y=0,Z=0)
    DriverItem=(Type=Bone,Name="thigh_l")
    FalloffRegionScaleFactors=(PositiveWidth=1,NegativeWidth=1,PositiveHeight=1,NegativeHeight=1)
    FalloffSize=0.33
    FlipHeightScaling=False
    FlipWidthScaling=False
    OptionalParentItem=(Type=Bone,Name="None")
    RotationOffset=(X=160,Y=30,Z=0)
   )
   > &RigUnit_OffsetTransformForItem(
    Item=(Type=Bone,Name="pcoket_B_l")
    OffsetTransform=(Rotation=(X=-0.0698,Y=0,Z=0,W=0.9976),Translation=(X=0,Y=0,Z=1.5),Scale3D=(X=1,Y=1,Z=1))
    bPropagateToChildren=True
    Weight=@n4.OutputParam
   )
   > &RigUnit_OffsetTransformForItem(
    Item=(Type=Bone,Name="pcoket_A_l")
    OffsetTransform=(Rotation=(X=0.0872,Y=0,Z=0,W=0.9962),Translation=(X=0,Y=0,Z=3),Scale3D=(X=1,Y=1,Z=1))
    bPropagateToChildren=True
    Weight=@n4.OutputParam
   )
   > n5 = &RigUnit_SphericalPoseReader(
    ActiveRegionScaleFactors=(PositiveWidth=1,NegativeWidth=1,PositiveHeight=1,NegativeHeight=1)
    ActiveRegionSize=0.1
    Debug=(bDrawDebug=false,bDraw2D=False,bDrawLocalAxes=False,DebugScale=15,DebugSegments=20,DebugThickness=0.1)
    DriverAxis=(X=1,Y=0,Z=0)
    DriverItem=(Type=Bone,Name="thigh_r")
    FalloffRegionScaleFactors=(PositiveWidth=1,NegativeWidth=1,PositiveHeight=1,NegativeHeight=1)
    FalloffSize=0.33
    FlipHeightScaling=False
    FlipWidthScaling=False
    OptionalParentItem=(Type=Bone,Name="None")
    RotationOffset=(X=-30,Y=20,Z=0)
   )
   > &RigUnit_OffsetTransformForItem(
    Item=(Type=Bone,Name="pcoket_B_r")
    OffsetTransform=(Rotation=(X=-0.0955,Y=0.0868,Z=0.0084,W=0.9916),Translation=(X=0,Y=0,Z=2.5),Scale3D=(X=1,Y=1,Z=1))
    bPropagateToChildren=True
    Weight=@n5.OutputParam
   )
   > n6 = &RigUnit_OffsetTransformForItem(
    Item=(Type=Bone,Name="pcoket_A_r")
    OffsetTransform=(Rotation=(X=0,Y=0,Z=-0.0958,W=0.9954),Translation=(X=0,Y=-1.5,Z=4),Scale3D=(X=1,Y=1,Z=1))
    bPropagateToChildren=True
    Weight=@n5.OutputParam
   )
   RigVMFunctionReturnNode(ExecutePin=@n6.ExecutePin)
  ]}
 CollapseNode_ContainedGraph {nodes [
   n1 = RigVMFunctionEntryNode()
   n2 = &RigUnit_SphericalPoseReader(
    ActiveRegionScaleFactors=(PositiveWidth=1.3,NegativeWidth=1,PositiveHeight=1,NegativeHeight=2)
    ActiveRegionSize=0.1
    Debug=(bDrawDebug=false,bDraw2D=False,bDrawLocalAxes=False,DebugScale=25,DebugSegments=20,DebugThickness=0.1)
    DriverAxis=(X=1,Y=0,Z=0)
    DriverItem=(Type=Bone,Name="upperarm_twist_01_l")
    FalloffRegionScaleFactors=(PositiveWidth=1,NegativeWidth=1,PositiveHeight=1,NegativeHeight=1)
    FalloffSize=0.3
    FlipHeightScaling=False
    FlipWidthScaling=False
    OptionalParentItem=(Type=Bone,Name="spine_04")
    RotationOffset=(X=0,Y=-120,Z=40)
    ExecutePin=@n1.ExecutePin
   )
   > &RigUnit_OffsetTransformForItem(
    Item=(Type=Bone,Name="armpit_l")
    OffsetTransform=(Rotation=(X=0,Y=0,Z=0,W=1),Translation=(X=4,Y=0,Z=0),Scale3D=(X=1,Y=1,Z=1))
    bPropagateToChildren=True
    Weight=@n2.OutputParam
   )
   > n3 = &RigUnit_SphericalPoseReader(
    ActiveRegionScaleFactors=(PositiveWidth=1,NegativeWidth=1,PositiveHeight=1,NegativeHeight=1)
    ActiveRegionSize=0.1
    Debug=(bDrawDebug=false,bDraw2D=False,bDrawLocalAxes=False,DebugScale=25,DebugSegments=20,DebugThickness=0.25)
    DriverAxis=(X=1,Y=0,Z=0)
    DriverItem=(Type=Bone,Name="upperarm_twist_01_l")
    FalloffRegionScaleFactors=(PositiveWidth=1,NegativeWidth=1,PositiveHeight=1.5,NegativeHeight=1)
    FalloffSize=0.3
    FlipHeightScaling=False
    FlipWidthScaling=False
    OptionalParentItem=(Type=Bone,Name="spine_04")
    RotationOffset=(X=65,Y=0,Z=0)
   )
   > &RigUnit_OffsetTransformForItem(
    Item=(Type=Bone,Name="armpit_l")
    OffsetTransform=(Rotation=(X=0,Y=0,Z=0,W=1),Translation=(X=5,Y=0,Z=0),Scale3D=(X=1,Y=1,Z=1))
    bPropagateToChildren=True
    Weight=@n3.OutputParam
   )
   > n4 = &RigUnit_SphericalPoseReader(
    ActiveRegionScaleFactors=(PositiveWidth=2,NegativeWidth=1,PositiveHeight=1.5,NegativeHeight=1)
    ActiveRegionSize=0.1
    Debug=(bDrawDebug=false,bDraw2D=False,bDrawLocalAxes=False,DebugScale=25,DebugSegments=20,DebugThickness=0.1)
    DriverAxis=(X=-1,Y=0,Z=0)
    DriverItem=(Type=Bone,Name="upperarm_twist_01_r")
    FalloffRegionScaleFactors=(PositiveWidth=1.5,NegativeWidth=1,PositiveHeight=1,NegativeHeight=1)
    FalloffSize=0.25
    FlipHeightScaling=False
    FlipWidthScaling=False
    OptionalParentItem=(Type=Bone,Name="spine_04")
    RotationOffset=(X=0,Y=30,Z=60)
   )
   > &RigUnit_OffsetTransformForItem(
    Item=(Type=Bone,Name="armpit_r")
    OffsetTransform=(Rotation=(X=0,Y=0,Z=0,W=1),Translation=(X=7,Y=0,Z=0),Scale3D=(X=1,Y=1,Z=1))
    bPropagateToChildren=True
    Weight=@n4.OutputParam
   )
   > n5 = &RigUnit_SphericalPoseReader(
    ActiveRegionScaleFactors=(PositiveWidth=1,NegativeWidth=1,PositiveHeight=1,NegativeHeight=1)
    ActiveRegionSize=0.1
    Debug=(bDrawDebug=false,bDraw2D=False,bDrawLocalAxes=False,DebugScale=25,DebugSegments=20,DebugThickness=0.25)
    DriverAxis=(X=-1,Y=0,Z=0)
    DriverItem=(Type=Bone,Name="upperarm_twist_01_r")
    FalloffRegionScaleFactors=(PositiveWidth=1,NegativeWidth=1.4,PositiveHeight=1,NegativeHeight=1)
    FalloffSize=0.3
    FlipHeightScaling=False
    FlipWidthScaling=False
    OptionalParentItem=(Type=Bone,Name="spine_04")
    RotationOffset=(X=-100,Y=-50,Z=45)
   )
   > n6 = &RigUnit_OffsetTransformForItem(
    Item=(Type=Bone,Name="armpit_r")
    OffsetTransform=(Rotation=(X=0,Y=0,Z=0,W=1),Translation=(X=7,Y=0,Z=0),Scale3D=(X=1,Y=1,Z=1))
    bPropagateToChildren=True
    Weight=@n5.OutputParam
   )
   RigVMFunctionReturnNode(ExecutePin=@n6.ExecutePin)
  ]}
 RigVMFunctionLibrary {links [], nodes []}
 RigVMModel {nodes [
   RigUnit_BeginExecution = &RigUnit_BeginExecution()
   FN_DeformRig_FNC = RigVMFunctionReferenceNode(
    DeformRig_ Curve="(deformRig_l_upperarm_up_alpha_72_#1=False,deformRig_r_upperarm_up_alpha_73_#2=False,deformRig_l_upperarm_fr_alpha_74_#3=False,deformRig_r_upperarm_fr_alpha_75_CB51B8 ...(truncated)"
    DeformRig_Bool="(upperarm_1_#4=True,elbow_3_#5=True,wrist_7_#6=True,groin_8_#7=True,glute_17_#8=True,knee_9_40C408324CA261 ...(truncated)"
    DeformRig_Debug=(upperarm_debug_18_#4=False,wrist_debug_20_#6=False,lat_debug_26_#9=False,groinGlute_debug_27_#7=False)
    DeformRig_Value="(upperarm_up_18_#10=(Rotation=(X=0,Y=0,Z=0,W=1),Translation=(X=0,Y=0,Z=8),Scale3D=(X=0.7,Y=1,Z=1)),upperarm_fwd_fr_19_#11=(Rotation=(X=0,Y=0,Z=0,W=1),Translation=(X=0,Y=10,Z=0),Scale3D=(X=0.5,Y=1,Z= ...(truncated)"
    ExecuteContext=@RigUnit_BeginExecution.ExecutePin
   )
   > Foots = RigVMCollapseNode()
   Forward belts = RigVMCollapseNode(ExecutePin=@Foots.ExecuteContext)
   Forward pockets = RigVMCollapseNode(ExecutePin=@"Forward belts".ExecutePin)
   &RigUnit_ParentConstraint(
    AdvancedSettings=(InterpolationType=Average,RotationOrderForFilter=XZY)
    Child=(Type=Bone,Name="belt_chain_A")
    Parents="((Item=(Type=Bone,Name=\"pelvis\"),Weight=0.1079),(Item=(Type=Bone,Name=\"spine_01\"),Weight=0.177),(Item=(Type=Bone,Name=\"spine_02\"),Weight=0.3218),(Item=(Type=Bone,Name=\"spine_03\"),Weight=0.3006),(Item=(Type=Bone,Name=\"spine_04\"),Weight=0.061),(Item=(Type=Bo ...(truncated)"
    Weight=1
    ExecutePin=@"Forward pockets".ExecutePin
   )
   > &RigUnit_ParentConstraint(
    AdvancedSettings=(InterpolationType=Average,RotationOrderForFilter=XZY)
    Child=(Type=Bone,Name="belt_chain_B")
    Parents="((Item=(Type=Bone,Name=\"pelvis\"),Weight=0.0241),(Item=(Type=Bone,Name=\"spine_01\"),Weight=0.0811),(Item=(Type=Bone,Name=\"spine_02\"),Weight=0.2198),(Item=(Type=Bone,Name=\"spine_03\"),Weight=0.3982),(Item=(Type=Bone,Name=\"spine_04\"),Weight=0.1732),(Item=(Type= ...(truncated)"
    Weight=1
   )
   > n1 = &RigUnit_ParentConstraint(
    AdvancedSettings=(InterpolationType=Average,RotationOrderForFilter=XZY)
    Child=(Type=Bone,Name="buckle_pnt")
    Parents="((Item=(Type=Bone,Name=\"spine_03\"),Weight=0.0406),(Item=(Type=Bone,Name=\"spine_04\"),Weight=0.1114),(Item=(Type=Bone,Name=\"spine_05\"),Weight=0.2542),(Item=(Type=Bone,Name=\"clavicle_l\"),Weight=0.323),(Item=(Type=Bone,Name=\"upperarm_twist_01_l\"),Weight=0.0532 ...(truncated)"
    Weight=1
   )
   Shoulder belt = RigVMCollapseNode(ExecutePin=@n1.ExecutePin)
   n2 = &RigUnit_SphericalPoseReader(
    ActiveRegionScaleFactors=(PositiveWidth=1.5,NegativeWidth=1,PositiveHeight=1,NegativeHeight=1)
    ActiveRegionSize=0.1
    Debug=(bDrawDebug=false,bDraw2D=False,bDrawLocalAxes=False,DebugScale=25,DebugSegments=20,DebugThickness=0.25)
    DriverAxis=(X=-1,Y=0,Z=0)
    DriverItem=(Type=Bone,Name="upperarm_r")
    FalloffRegionScaleFactors=(PositiveWidth=1,NegativeWidth=2.4,PositiveHeight=1.9,NegativeHeight=1)
    FalloffSize=0.2
    FlipHeightScaling=False
    FlipWidthScaling=False
    OptionalParentItem=(Type=Bone,Name="spine_05")
    RotationOffset=(X=-110,Y=0,Z=50)
    ExecutePin=@"Shoulder belt".ExecutePin
   )
   > &RigUnit_OffsetTransformForItem(
    Item=(Type=Bone,Name="deform_upperarm_fr_base_r")
    OffsetTransform=(Rotation=(X=0,Y=0,Z=0.2588,W=0.9659),Translation=(X=0,Y=0,Z=0),Scale3D=(X=1,Y=1,Z=1))
    bPropagateToChildren=True
    Weight=@n2.OutputParam
   )
   > &RigUnit_OffsetTransformForItem(
    Item=(Type=Bone,Name="deform_upperarm_fr_r")
    OffsetTransform=(Rotation=(X=0,Y=0,Z=-0.7071,W=0.7071),Translation=(X=0,Y=-6,Z=0),Scale3D=(X=2,Y=1,Z=1))
    bPropagateToChildren=True
    Weight=@n2.OutputParam
   )
   > &RigUnit_OffsetTransformForItem(
    Item=(Type=Bone,Name="deform_upperarm_fr_dup_r")
    OffsetTransform=(Rotation=(X=0,Y=0,Z=0,W=1),Translation=(X=4,Y=0,Z=0),Scale3D=(X=1,Y=1,Z=1))
    bPropagateToChildren=True
    Weight=@n2.OutputParam
   )
   > n3 = &RigUnit_SphericalPoseReader(
    ActiveRegionScaleFactors=(PositiveWidth=1,NegativeWidth=1,PositiveHeight=1,NegativeHeight=1)
    ActiveRegionSize=0.3
    Debug=(bDrawDebug=true,bDraw2D=False,bDrawLocalAxes=False,DebugScale=25,DebugSegments=20,DebugThickness=0.25)
    DriverAxis=(X=1,Y=0,Z=0)
    DriverItem=(Type=Bone,Name="spine_05")
    FalloffRegionScaleFactors=(PositiveWidth=1,NegativeWidth=1,PositiveHeight=1,NegativeHeight=1)
    FalloffSize=0.5
    FlipHeightScaling=False
    FlipWidthScaling=False
    OptionalParentItem=(Type=Bone,Name="spine_03")
    RotationOffset=(X=90,Y=0,Z=0)
   )
   > n4 = &RigUnit_OffsetTransformForItem(
    Item=(Type=Bone,Name="belt_chain_6")
    OffsetTransform=(Rotation=(X=0,Y=0,Z=0,W=1),Translation=(X=3,Y=0,Z=0),Scale3D=(X=1,Y=1,Z=1))
    bPropagateToChildren=True
    Weight=@n3.OutputParam
   )
   Pockets = RigVMCollapseNode(ExecutePin=@n4.ExecutePin)
   n5 = &RigUnit_SphericalPoseReader(
    ActiveRegionScaleFactors=(PositiveWidth=1,NegativeWidth=1,PositiveHeight=1,NegativeHeight=1)
    ActiveRegionSize=0.2
    Debug=(bDrawDebug=false,bDraw2D=False,bDrawLocalAxes=False,DebugScale=25,DebugSegments=20,DebugThickness=0.25)
    DriverAxis=(X=0,Y=1,Z=0)
    DriverItem=(Type=Bone,Name="spine_05")
    FalloffRegionScaleFactors=(PositiveWidth=1,NegativeWidth=1,PositiveHeight=1,NegativeHeight=1)
    FalloffSize=0.5
    FlipHeightScaling=False
    FlipWidthScaling=False
    OptionalParentItem=(Type=Bone,Name="spine_02")
    RotationOffset=(X=0,Y=0,Z=0)
    ExecutePin=@Pockets.ExecutePin
   )
   > RigUnit_OffsetTransformForItem = &RigUnit_OffsetTransformForItem(
    Item=(Type=Bone,Name="belt_chain_7")
    OffsetTransform=(Rotation=(X=0,Y=0,Z=0,W=1),Translation=(X=4,Y=0,Z=0),Scale3D=(X=1,Y=1,Z=1))
    bPropagateToChildren=True
    Weight=@n5.OutputParam
   )
   > n6 = &RigUnit_SphericalPoseReader(
    ActiveRegionScaleFactors=(PositiveWidth=1,NegativeWidth=1,PositiveHeight=1,NegativeHeight=1)
    ActiveRegionSize=0.2
    Debug=(bDrawDebug=false,bDraw2D=False,bDrawLocalAxes=False,DebugScale=25,DebugSegments=20,DebugThickness=0.25)
    DriverAxis=(X=0,Y=1,Z=0)
    DriverItem=(Type=Bone,Name="spine_05")
    FalloffRegionScaleFactors=(PositiveWidth=1,NegativeWidth=1,PositiveHeight=1,NegativeHeight=1)
    FalloffSize=0.5
    FlipHeightScaling=False
    FlipWidthScaling=False
    OptionalParentItem=(Type=Bone,Name="spine_03")
    RotationOffset=(X=180,Y=0,Z=0)
   )
   > n7 = &RigUnit_OffsetTransformForItem(
    Item=(Type=Bone,Name="belt_chain_7")
    OffsetTransform=(Rotation=(X=0,Y=0,Z=0,W=1),Translation=(X=2,Y=0,Z=0),Scale3D=(X=1,Y=1,Z=1))
    bPropagateToChildren=True
    Weight=@n6.OutputParam
   )
   Armpits = RigVMCollapseNode(ExecutePin=@n7.ExecutePin)
  ]}
]
hierarchy [
 root {loc=<0 0 0>, rot=<0 0 0 1>}
 pelvis {parent=root, loc=<0 -0.7071 93.1191>, rot=<0.0127 0.707 -0.0127 -0.707>}
 spine_01 {parent=pelvis, loc=<6.7494 -0.6388 0>, rot=<0 0 0.034 -0.9994>}
 spine_02 {parent=spine_01, loc=<7.2518 -0.035 0>, rot=<0 0 0.0238 -0.9997>}
 spine_03 {parent=spine_02, loc=<6.0104 0 0>, rot=<0 0 0.0149 -0.9999>}
 spine_04 {parent=spine_03, loc=<8.1024 0 0>, rot=<0 0 -0.0132 -0.9999>}
 spine_05 {parent=spine_04, loc=<10.0146 0 0>, rot=<0 0 -0.0015 -1>}
 neck_01 {parent=spine_05, loc=<11.3138 0.1529 0>, rot=<0 0 -0.2003 -0.9797>}
 neck_02 {parent=neck_01, loc=<4.5627 0 0>, rot=<0 0 0.0504 -0.9987>}
 head {parent=neck_02, loc=<5.3914 0 0>, rot=<0 0 0.115 -0.9934>}
 FACIAL_C_FacialRoot {parent=head, loc=<2.1524 -0.746 -0.0002>, rot=<0.4879 -0.5118 -0.4879 -0.5118>}
 FACIAL_C_Forehead {parent=FACIAL_C_FacialRoot, loc=<-0.0006 -3.4793 10.0431>, rot=<-0.0469 0.0001 -0.0001 -0.9989>}
 FACIAL_L_ForeheadIn {parent=FACIAL_C_FacialRoot, loc=<1.4147 -3.4637 9.9646>, rot=<-0.0642 -0.0695 0.004 -0.9955>}
 FACIAL_R_ForeheadIn {parent=FACIAL_C_FacialRoot, loc=<-1.4159 -3.4637 9.9652>, rot=<-0.0642 0.0698 -0.0041 -0.9955>}
 FACIAL_L_ForeheadMid {parent=FACIAL_C_FacialRoot, loc=<2.7216 -3.6965 9.5833>, rot=<-0.0744 -0.1754 -0.0213 -0.9814>}
 FACIAL_R_ForeheadMid {parent=FACIAL_C_FacialRoot, loc=<-2.7233 -3.6968 9.5851>, rot=<-0.0745 0.1757 0.0212 -0.9814>}
 FACIAL_L_ForeheadOut {parent=FACIAL_C_FacialRoot, loc=<3.9735 -3.5835 8.8402>, rot=<-0.1247 -0.2552 -0.0801 -0.9555>}
 FACIAL_R_ForeheadOut {parent=FACIAL_C_FacialRoot, loc=<-3.9754 -3.5843 8.8411>, rot=<-0.1248 0.2555 0.08 -0.9554>}
 FACIAL_L_EyesackUpper {parent=FACIAL_C_FacialRoot, loc=<2.7692 -2.6373 9.9301>, rot=<0.0453 -0.0839 -0.0649 -0.9933>}
 FACIAL_R_EyesackUpper {parent=FACIAL_C_FacialRoot, loc=<-2.7652 -2.6358 9.9291>, rot=<0.0453 0.0842 0.0648 -0.9933>}
 FACIAL_L_EyelidUpperFurrow {parent=FACIAL_C_FacialRoot, loc=<2.6563 -2.5036 9.6811>, rot=<-0.0335 -0.1504 -0.0802 -0.9848>}
 FACIAL_R_EyelidUpperFurrow {parent=FACIAL_C_FacialRoot, loc=<-2.6506 -2.504 9.677>, rot=<-0.0335 0.1507 0.0801 -0.9848>}
 FACIAL_L_EyelidUpperA {parent=FACIAL_C_FacialRoot, loc=<2.6892 -2.0537 8.1276>, rot=<-0.1513 0.0001 -0.0001 -0.9885>}
 FACIAL_R_EyelidUpperA {parent=FACIAL_C_FacialRoot, loc=<-2.6864 -2.0518 8.1322>, rot=<-0.1513 0.0001 -0.0001 -0.9885>}
 FACIAL_L_EyelidLowerA {parent=FACIAL_C_FacialRoot, loc=<2.6892 -2.0537 8.1276>, rot=<0.2806 0.0002 0 -0.9598>}
 FACIAL_R_EyelidLowerA {parent=FACIAL_C_FacialRoot, loc=<-2.6864 -2.0518 8.1322>, rot=<0.2806 0.0002 0 -0.9598>}
 FACIAL_L_EyeCornerInner {parent=FACIAL_C_FacialRoot, loc=<1.7731 -1.7742 8.9621>, rot=<-0.0035 0.0001 0 -1>}
 FACIAL_R_EyeCornerInner {parent=FACIAL_C_FacialRoot, loc=<-1.7763 -1.774 8.9545>, rot=<-0.0035 0.0001 0 -1>}
 FACIAL_L_EyeCornerOuter {parent=FACIAL_C_FacialRoot, loc=<3.7822 -1.8453 8.6203>, rot=<-0.0032 -0.3825 -0.0014 -0.924>}
 FACIAL_R_EyeCornerOuter {parent=FACIAL_C_FacialRoot, loc=<-3.7763 -1.8343 8.6149>, rot=<-0.0033 0.3827 0.0013 -0.9239>}
 FACIAL_L_EyesackLower {parent=FACIAL_C_FacialRoot, loc=<2.8353 -1.0331 8.5782>, rot=<-0.0258 -0.1395 -0.0228 -0.9896>}
 FACIAL_R_EyesackLower {parent=FACIAL_C_FacialRoot, loc=<-2.8337 -1.0311 8.5714>, rot=<-0.0258 0.1398 0.0227 -0.9896>}
 FACIAL_L_CheekInner {parent=FACIAL_C_FacialRoot, loc=<2.0096 -0.0047 8.5952>, rot=<-0.1266 -0.0872 -0.1975 -0.9682>}
 FACIAL_R_CheekInner {parent=FACIAL_C_FacialRoot, loc=<-2.0102 -0.0059 8.5938>, rot=<-0.1267 0.0874 0.1974 -0.9682>}
 FACIAL_L_CheekOuter {parent=FACIAL_C_FacialRoot, loc=<3.7017 0.3894 7.2521>, rot=<0.052 -0.3772 0.099 -0.9194>}
 FACIAL_R_CheekOuter {parent=FACIAL_C_FacialRoot, loc=<-3.7014 0.3894 7.2484>, rot=<0.0519 0.3774 -0.0991 -0.9193>}
 FACIAL_L_NasolabialBulge {parent=FACIAL_C_FacialRoot, loc=<3.7412 3.74 7.152>, rot=<0.0547 -0.3773 0.0224 -0.9242>}
 FACIAL_R_NasolabialBulge {parent=FACIAL_C_FacialRoot, loc=<-3.7398 3.7428 7.1489>, rot=<0.0546 0.3776 -0.0225 -0.9241>}
 FACIAL_L_NasolabialFurrow {parent=FACIAL_C_FacialRoot, loc=<2.966 2.4355 9.2021>, rot=<-0.0365 -0.1244 -0.3641 -0.9223>}
 FACIAL_R_NasolabialFurrow {parent=FACIAL_C_FacialRoot, loc=<-2.9643 2.442 9.2016>, rot=<-0.0366 0.1247 0.364 -0.9223>}
 FACIAL_L_CheekLower {parent=FACIAL_C_FacialRoot, loc=<4.2093 3.4478 6.3666>, rot=<0.0314 -0.4269 0.0562 -0.902>}
 FACIAL_R_CheekLower {parent=FACIAL_C_FacialRoot, loc=<-4.2068 3.4498 6.3628>, rot=<0.0314 0.4272 -0.0563 -0.9019>}
 FACIAL_C_Nose {parent=FACIAL_C_FacialRoot, loc=<-0.0016 -0.2383 10.6765>, rot=<-0.1338 0.0001 -0.0001 -0.991>}
 FACIAL_C_NoseLower {parent=FACIAL_C_Nose, loc=<0.0015 2.7185 -0.6187>, rot=<0.1305 0 0 -0.9914>}
 FACIAL_L_Nostril {parent=FACIAL_C_Nose, loc=<1.1251 1.836 -0.5621>, rot=<0.0173 0.1305 0.0023 -0.9913>}
 FACIAL_R_Nostril {parent=FACIAL_C_Nose, loc=<-1.1266 1.8348 -0.5667>, rot=<0.0173 -0.1305 -0.0023 -0.9913>}
 FACIAL_C_NoseTip {parent=FACIAL_C_Nose, loc=<0.0006 1.9402 0.6457>, rot=<0.0872 0 0 -0.9962>}
 FACIAL_C_MouthUpper {parent=FACIAL_C_FacialRoot, loc=<-0.0008 3.9433 8.0866>, rot=<-0.0035 0.0001 0 -1>}
 FACIAL_C_LipUpper {parent=FACIAL_C_MouthUpper, loc=<0.0001 -0.2935 2.2258>, rot=<-0.0872 0 0 -0.9962>}
 FACIAL_L_LipUpper {parent=FACIAL_C_MouthUpper, loc=<1.2157 -0.1933 1.9949>, rot=<-0.0778 -0.1601 -0.0379 -0.9833>}
 FACIAL_R_LipUpper {parent=FACIAL_C_MouthUpper, loc=<-1.2146 -0.1968 1.9959>, rot=<-0.0778 0.1601 0.0379 -0.9833>}
 FACIAL_L_LipUpperOuter {parent=FACIAL_C_MouthUpper, loc=<2.0683 0.0435 1.3819>, rot=<-0.0984 -0.2293 -0.0983 -0.9634>}
 FACIAL_R_LipUpperOuter {parent=FACIAL_C_MouthUpper, loc=<-2.0716 0.0372 1.3803>, rot=<-0.0984 0.2293 0.0983 -0.9634>}
 FACIAL_L_LipCorner {parent=FACIAL_C_MouthUpper, loc=<2.7171 0.3265 0.7683>, rot=<0 -0.2078 0 -0.9782>}
 FACIAL_R_LipCorner {parent=FACIAL_C_MouthUpper, loc=<-2.7424 0.3277 0.7658>, rot=<0 0.2078 0 -0.9782>}
 FACIAL_L_Masseter {parent=FACIAL_C_FacialRoot, loc=<5.9574 1.2519 5.0076>, rot=<0.0659 0.113 0.0082 -0.9914>}
 FACIAL_R_Masseter {parent=FACIAL_C_FacialRoot, loc=<-5.9575 1.2543 5.0034>, rot=<0.0659 -0.1127 -0.0083 -0.9914>}
 FACIAL_C_UnderChin {parent=FACIAL_C_FacialRoot, loc=<-0.0008 8.4545 5.665>, rot=<0.4968 0.0002 0 -0.8679>}
 FACIAL_L_UnderChin {parent=FACIAL_C_FacialRoot, loc=<3.2023 7.2728 4.2665>, rot=<0.3996 -0.355 0.0811 -0.8413>}
 FACIAL_R_UnderChin {parent=FACIAL_C_FacialRoot, loc=<-3.2043 7.2778 4.2619>, rot=<0.3996 0.3553 -0.081 -0.8412>}
 FACIAL_C_LowerLipRotation {parent=FACIAL_C_FacialRoot, loc=<0.0013 1.7692 3.1129>, rot=<0.3387 0.0002 0 -0.9409>}
 FACIAL_C_MouthLower {parent=FACIAL_C_LowerLipRotation, loc=<-0.0005 -1.4951 5.2182>, rot=<-0.342 0 0 -0.9397>}
 FACIAL_C_LipLower {parent=FACIAL_C_MouthLower, loc=<0.0004 1.2226 1.8564>, rot=<0.0872 0 0 -0.9962>}
 FACIAL_L_LipLower {parent=FACIAL_C_MouthLower, loc=<1.0409 1.1409 1.7301>, rot=<0.0927 -0.1294 0.0272 -0.9869>}
 FACIAL_R_LipLower {parent=FACIAL_C_MouthLower, loc=<-1.0419 1.1406 1.7293>, rot=<0.0927 0.1294 -0.0272 -0.9869>}
 FACIAL_L_LipLowerOuter {parent=FACIAL_C_MouthLower, loc=<1.9646 0.8172 1.2295>, rot=<0.0913 -0.2321 0.0318 -0.9679>}
 FACIAL_R_LipLowerOuter {parent=FACIAL_C_MouthLower, loc=<-1.9678 0.8191 1.2305>, rot=<0.0913 0.2321 -0.0318 -0.9679>}
 FACIAL_C_Jaw {parent=FACIAL_C_FacialRoot, loc=<0.0013 1.7692 3.1129>, rot=<0.3387 0.0002 0 -0.9409>}
 FACIAL_C_Jawline {parent=FACIAL_C_Jaw, loc=<-0.002 2.2893 7.9726>, rot=<0.2586 0 0 -0.966>}
 FACIAL_L_ChinSide {parent=FACIAL_C_Jaw, loc=<3.0726 0.3064 7.286>, rot=<-0.2208 -0.3738 0.0499 -0.8995>}
 FACIAL_R_ChinSide {parent=FACIAL_C_Jaw, loc=<-3.0744 0.3176 7.2862>, rot=<-0.2208 0.3738 -0.0499 -0.8995>}
 FACIAL_C_Chin {parent=FACIAL_C_Jaw, loc=<-0.0035 0.7786 7.8908>, rot=<-0.2591 0 0 -0.9659>}
 FACIAL_L_Ear {parent=FACIAL_C_FacialRoot, loc=<6.141 -0.4568 2.3461>, rot=<-0.0298 -0.0004 -0.0262 -0.9992>}
 FACIAL_R_Ear {parent=FACIAL_C_FacialRoot, loc=<-6.1434 -0.4559 2.3491>, rot=<-0.0298 0.0007 0.0261 -0.9992>}
 FACIAL_L_Jawline {parent=FACIAL_C_FacialRoot, loc=<4.3921 5.326 5.4511>, rot=<0.2059 0.1581 -0.0541 -0.9642>}
 FACIAL_R_Jawline {parent=FACIAL_C_FacialRoot, loc=<-4.3924 5.327 5.4442>, rot=<0.2059 -0.1578 0.0541 -0.9642>}
 FACIAL_C_NoseBridge {parent=FACIAL_C_FacialRoot, loc=<-0.0013 -2.0953 10.8348>, rot=<0.0403 0.0001 0 -0.9992>}
 FACIAL_L_EyelidUpperB {parent=FACIAL_C_FacialRoot, loc=<2.6892 -2.0537 8.1276>, rot=<-0.2284 0.0001 -0.0001 -0.9736>}
 FACIAL_L_EyelidLowerB {parent=FACIAL_C_FacialRoot, loc=<2.6892 -2.0537 8.1276>, rot=<0.314 0.0002 0 -0.9494>}
 FACIAL_R_EyelidLowerB {parent=FACIAL_C_FacialRoot, loc=<-2.6864 -2.0518 8.1322>, rot=<0.3222 0.0002 0 -0.9467>}
 FACIAL_R_EyelidUpperB {parent=FACIAL_C_FacialRoot, loc=<-2.6864 -2.0518 8.1322>, rot=<-0.2284 0.0001 -0.0001 -0.9736>}
 attach_cape {parent=spine_05, loc=<9.017 -11.714 0>, rot=<0.0616 -0.7044 -0.0616 -0.7044>}
 cape {parent=attach_cape, type=Null, loc=<0 0 0>, rot=<0 0 0 1>}
 attach_backpack {parent=spine_05, loc=<-0.123 -13.473 0>, rot=<-0.0093 -0.707 0.0093 -0.707>}
 Backpack_BR {parent=attach_backpack, type=Null, loc=<0 0 0>, rot=<0 0 0 1>}
 clavicle_l {parent=spine_05, loc=<4.3575 -0.1466 -2.3223>, rot=<0.0465 -0.7289 -0.009 -0.6829>}
 upperarm_l {parent=clavicle_l, loc=<15.5111 0 0>, rot=<0.0226 -0.3234 -0.0088 -0.9459>}
 lowerarm_l {parent=upperarm_l, loc=<26.6514 0 0>, rot=<0.0391 -0.0099 -0.2453 -0.9686>}
 hand_l {parent=lowerarm_l, loc=<22.3491 0 0>, rot=<-0.7081 -0.0507 0.0975 -0.6975>}
 thumb_01_l {parent=hand_l, loc=<1.3118 -0.3191 -2.459>, rot=<0.6546 -0.4042 0.0664 -0.6354>}
 thumb_02_l {parent=thumb_01_l, loc=<5.9336 -0.0448 0.0401>, rot=<-0.0037 0.0213 0.298 -0.9543>}
 thumb_03_l {parent=thumb_02_l, loc=<2.2223 0.3682 -0.0907>, rot=<0.0118 -0.013 0.0066 -0.9998>}
 index_metacarpal_l {parent=hand_l, loc=<1.6535 0.0183 -2.96>, rot=<0.0158 -0.0124 -0.0173 -0.9996>}
 index_01_l {parent=index_metacarpal_l, loc=<8.3008 0.0357 -0.0524>, rot=<-0.0044 0.0485 0.0924 -0.9945>}
 index_02_l {parent=index_01_l, loc=<3.3587 -0.3004 0.0452>, rot=<-0.002 0.0027 0.3225 -0.9466>}
 index_03_l {parent=index_02_l, loc=<2.5578 -0.123 -0.0252>, rot=<-0.002 0.0067 0.2558 -0.9667>}
 middle_metacarpal_l {parent=hand_l, loc=<2.1856 -0.1725 -0.5984>, rot=<0.006 0.0168 -0.0329 -0.9993>}
 middle_01_l {parent=middle_metacarpal_l, loc=<7.3275 -0.2447 0.0129>, rot=<-0.0208 0.0042 0.0858 -0.9961>}
 middle_02_l {parent=middle_01_l, loc=<4.0123 0.1401 0.1175>, rot=<-0.0218 0.0258 0.3512 -0.9357>}
 middle_03_l {parent=middle_02_l, loc=<2.6077 0.0534 0.1373>, rot=<0.0082 0.0188 0.1708 -0.9851>}
 ring_metacarpal_l {parent=hand_l, loc=<2.2074 -0.3625 1.3638>, rot=<-0.0284 0.0722 -0.0458 -0.9959>}
 ring_01_l {parent=ring_metacarpal_l, loc=<7.0295 -0.6197 -0.1446>, rot=<-0.0489 -0.0252 0.1892 -0.9804>}
 ring_02_l {parent=ring_01_l, loc=<3.6174 0.4647 -0.1027>, rot=<0 0 0.2806 -0.9598>}
 ring_03_l {parent=ring_02_l, loc=<2.4152 0.114 0.0199>, rot=<0 0 0.1189 -0.9929>}
 pinky_metacarpal_l {parent=hand_l, loc=<1.7616 -0.9753 3.3814>, rot=<-0.0397 0.0814 -0.069 -0.9935>}
 pinky_01_l {parent=pinky_metacarpal_l, loc=<7.1709 -0.4831 -0.1094>, rot=<-0.0654 -0.0035 0.2242 -0.9723>}
 pinky_02_l {parent=pinky_01_l, loc=<2.8274 0.0893 -0.1043>, rot=<0.0037 0.0011 0.2854 -0.9584>}
 pinky_03_l {parent=pinky_02_l, loc=<2.2046 0.0307 -0.0078>, rot=<-0.0002 0.0098 0.1445 -0.9895>}
 deform_wrist_dn_base_l {parent=hand_l, loc=<0 -2.2342 0>, rot=<0 0 0 1>}
 deform_wrist_dn_l {parent=deform_wrist_dn_base_l, loc=<0 0 0>, rot=<0 0 0 1>}
 deform_wrist_up_base_l {parent=hand_l, loc=<0 2.2682 0>, rot=<0 0 0 1>}
 deform_wrist_up_l {parent=deform_wrist_up_base_l, loc=<0 0 0>, rot=<0 0 0 1>}
 weapon_l {parent=hand_l, loc=<-0.1974 -3.4111 0.3811>, rot=<0 0 0 1>}
 lowerarm_twist_02_l {parent=lowerarm_l, loc=<9.6199 0 0>, rot=<0 0 0 -1>}
 lowerarm_twist_01_l {parent=lowerarm_l, loc=<17.8549 0 0>, rot=<0 0 0 -1>}
 deform_elbow_in_base_l {parent=lowerarm_l, loc=<0.2112 5.259 0.8274>, rot=<0 0 0 -1>}
 deform_elbow_in_l {parent=deform_elbow_in_base_l, loc=<0 0 0>, rot=<0 0 0 1>}
 deform_elbow_out_base_l {parent=lowerarm_l, loc=<0 -4 0>, rot=<0 0 0 -1>}
 deform_elbow_out_l {parent=deform_elbow_out_base_l, loc=<0 0 0>, rot=<0 0 0 1>}
 upperarm_twist_01_l {parent=upperarm_l, loc=<0.0041 0 0>, rot=<0 0 0 -1>}
 deform_upperarm_bk_base_l {parent=upperarm_twist_01_l, loc=<1.0813 -7.2193 -1.9453>, rot=<0 0 0 1>}
 deform_upperarm_bk_l {parent=deform_upperarm_bk_base_l, loc=<0 0 0>, rot=<0 0 0 1>}
 deform_upperarm_fr_base_l {parent=upperarm_twist_01_l, loc=<-0.3691 7.0506 -2.765>, rot=<0 0 0 1>}
 deform_upperarm_fr_l {parent=deform_upperarm_fr_base_l, loc=<0 0 0>, rot=<0 0 0 1>}
 deform_upperarm_up_base_l {parent=upperarm_twist_01_l, loc=<-3.2216 -1.1391 3.3311>, rot=<0 0 0 1>}
 deform_upperarm_up_l {parent=deform_upperarm_up_base_l, loc=<0 0 0>, rot=<0 0 0 1>}
 shoulder_l_pnt {parent=deform_upperarm_up_base_l, loc=<-0.79 4.21 0.47>, rot=<0 0 0 1>}
 upperarm_twist_02_l {parent=upperarm_l, loc=<13.7941 0 0>, rot=<0 0 0 -1>}
 clavicle_r {parent=spine_05, loc=<4.3575 -0.1466 2.3223>, rot=<0.7289 0.0465 -0.6829 0.009>}
 upperarm_r {parent=clavicle_r, loc=<-15.5111 0 0>, rot=<0.0226 -0.3234 -0.0088 -0.9459>}
 lowerarm_r {parent=upperarm_r, loc=<-26.6514 0 0>, rot=<0.0391 -0.0099 -0.2453 -0.9686>}
 hand_r {parent=lowerarm_r, loc=<-22.3491 0 0>, rot=<-0.7081 -0.0507 0.0975 -0.6975>}
 thumb_01_r {parent=hand_r, loc=<-1.3119 0.3192 2.459>, rot=<0.6546 -0.4042 0.0664 -0.6354>}
 thumb_02_r {parent=thumb_01_r, loc=<-5.9336 0.0448 -0.0401>, rot=<-0.0037 0.0213 0.298 -0.9543>}
 thumb_03_r {parent=thumb_02_r, loc=<-2.2223 -0.3682 0.0907>, rot=<0.0118 -0.013 0.0066 -0.9998>}
 index_metacarpal_r {parent=hand_r, loc=<-1.6534 -0.0181 2.9601>, rot=<0.0158 -0.0124 -0.0173 -0.9996>}
 index_01_r {parent=index_metacarpal_r, loc=<-8.3008 -0.0357 0.0524>, rot=<-0.0044 0.0485 0.0924 -0.9945>}
 index_02_r {parent=index_01_r, loc=<-3.3589 0.3003 -0.0453>, rot=<-0.002 0.0027 0.3225 -0.9466>}
 index_03_r {parent=index_02_r, loc=<-2.5578 0.123 0.0252>, rot=<-0.002 0.0067 0.2558 -0.9667>}
 middle_metacarpal_r {parent=hand_r, loc=<-2.1856 0.1726 0.5984>, rot=<0.006 0.0168 -0.0329 -0.9993>}
 middle_01_r {parent=middle_metacarpal_r, loc=<-7.3273 0.2447 -0.0129>, rot=<-0.0208 0.0042 0.0858 -0.9961>}
 middle_02_r {parent=middle_01_r, loc=<-4.0142 -0.1411 -0.07>, rot=<-0.0217 0.0259 0.3512 -0.9357>}
 middle_03_r {parent=middle_02_r, loc=<-2.6074 -0.0534 -0.1312>, rot=<0.0083 0.0188 0.1708 -0.9851>}
 ring_metacarpal_r {parent=hand_r, loc=<-2.2073 0.3627 -1.3638>, rot=<-0.0284 0.0722 -0.0458 -0.9959>}
 ring_01_r {parent=ring_metacarpal_r, loc=<-7.0297 0.6197 0.1446>, rot=<-0.0489 -0.0252 0.1892 -0.9804>}
 ring_02_r {parent=ring_01_r, loc=<-3.6171 -0.4648 0.1027>, rot=<0 0 0.2806 -0.9598>}
 ring_03_r {parent=ring_02_r, loc=<-2.4152 -0.114 -0.0199>, rot=<0 0 0.1189 -0.9929>}
 pinky_metacarpal_r {parent=hand_r, loc=<-1.7617 0.9753 -3.3814>, rot=<-0.0397 0.0814 -0.069 -0.9935>}
 pinky_01_r {parent=pinky_metacarpal_r, loc=<-7.1707 0.4832 0.1094>, rot=<-0.0654 -0.0035 0.2242 -0.9723>}
 pinky_02_r {parent=pinky_01_r, loc=<-2.8274 -0.0893 0.1043>, rot=<0.0037 0.0011 0.2854 -0.9584>}
 pinky_03_r {parent=pinky_02_r, loc=<-2.2045 -0.031 0.0078>, rot=<-0.0002 0.0098 0.1445 -0.9895>}
 deform_wrist_dn_base_r {parent=hand_r, loc=<-0.0003 2.2346 -0.0001>, rot=<0 0 0 1>}
 deform_wrist_dn_r {parent=deform_wrist_dn_base_r, loc=<0 0 0>, rot=<0 0 0 1>}
 deform_wrist_up_base_r {parent=hand_r, loc=<0.0004 -2.2683 0.0001>, rot=<0 0 0 1>}
 deform_wrist_up_r {parent=deform_wrist_up_base_r, loc=<0 0 0>, rot=<0 0 0 1>}
 weapon_r {parent=hand_r, loc=<0.1975 3.4112 -0.3811>, rot=<0 0 0 -1>}
 lowerarm_twist_02_r {parent=lowerarm_r, loc=<-9.6199 0 0>, rot=<0 0 0 -1>}
 lowerarm_twist_01_r {parent=lowerarm_r, loc=<-17.8549 0 0>, rot=<0 0 0 -1>}
 deform_elbow_in_base_r {parent=lowerarm_r, loc=<-0.2111 -5.2591 -0.8275>, rot=<0 0 0 -1>}
 deform_elbow_in_r {parent=deform_elbow_in_base_r, loc=<-0.0001 0 0.0002>, rot=<0 0 0 1>}
 deform_elbow_out_base_r {parent=lowerarm_r, loc=<0 4 -0.0001>, rot=<0 0 0 -1>}
 deform_elbow_out_r {parent=deform_elbow_out_base_r, loc=<0 0 0>, rot=<0 0 0 1>}
 upperarm_twist_01_r {parent=upperarm_r, loc=<-0.0041 0 0>, rot=<0 0 0 -1>}
 deform_upperarm_bk_base_r {parent=upperarm_twist_01_r, loc=<-1.0814 7.2192 1.9456>, rot=<0 0 0 1>}
 deform_upperarm_bk_r {parent=deform_upperarm_bk_base_r, loc=<0 0 0>, rot=<0 0 0 1>}
 deform_upperarm_fr_base_r {parent=upperarm_twist_01_r, loc=<0.3689 -7.0506 2.7652>, rot=<0 0 0 1>}
 deform_upperarm_fr_r {parent=deform_upperarm_fr_base_r, loc=<0 0 0>, rot=<0 0 0 1>}
 deform_upperarm_fr_dup_r {parent=deform_upperarm_fr_r, loc=<0 0 0>, rot=<0 0 0 1>}
 deform_upperarm_up_base_r {parent=upperarm_twist_01_r, loc=<3.2214 1.1391 -3.3308>, rot=<0 0 0 1>}
 deform_upperarm_up_r {parent=deform_upperarm_up_base_r, loc=<0 0 0>, rot=<0 0 0 -1>}
 upperarm_twist_02_r {parent=upperarm_r, loc=<-13.7941 0 0>, rot=<0 0 0 -1>}
 belt_chain_root {parent=spine_05, loc=<14.6032 0.6123 -9.6943>, rot=<0.0284 -0.7065 -0.0284 -0.7065>}
 belt_chain_fr_1 {parent=belt_chain_root, loc=<-1.2487 6.7977 -3.7964>, rot=<0.192 -0.4584 -0.6805 -0.5384>}
 belt_chain_fr_2 {parent=belt_chain_fr_1, loc=<11.5548 0 0>, rot=<-0.031 -0.1918 -0.074 -0.9782>}
 belt_chain_fr_3 {parent=belt_chain_fr_2, loc=<12.5774 0 0>, rot=<-0.0039 -0.1616 -0.0978 -0.982>}
 belt_chain_fr_4 {parent=belt_chain_fr_3, loc=<7.7312 0 0>, rot=<0.0298 -0.2925 -0.0665 -0.9535>}
 belt_chain_fr_5 {parent=belt_chain_fr_4, loc=<9.1227 0 0>, rot=<0.1175 -0.2987 -0.0714 -0.9444>}
 belt_chain_fr_6 {parent=belt_chain_fr_5, loc=<6.6721 0 0>, rot=<0 0 0 1>}
 belt_chain_bk_1 {parent=belt_chain_root, loc=<-1.0269 -6.7143 -1.0996>, rot=<-0.1682 -0.4856 0.6868 -0.514>}
 belt_chain_bk_2 {parent=belt_chain_bk_1, loc=<7.6598 0 0>, rot=<0.0166 -0.1858 0.0408 -0.9816>}
 belt_chain_bk_3 {parent=belt_chain_bk_2, loc=<10.7466 0 0>, rot=<0.0011 -0.1681 0.0354 -0.9851>}
 belt_chain_bk_4 {parent=belt_chain_bk_3, loc=<12.8497 0 0>, rot=<-0.0091 -0.2143 0.0247 -0.9764>}
 belt_chain_bk_5 {parent=belt_chain_bk_4, loc=<9.5683 0 0>, rot=<0.0563 -0.3813 -0.0392 -0.9219>}
 belt_chain_bk_6 {parent=belt_chain_bk_5, loc=<8.5938 0 0>, rot=<0 0 0 1>}
 belt_chain_shoulder_base {parent=spine_05, loc=<14.693 -0.5045 -9.6943>, rot=<0.006 -0.1491 -0.0397 -0.988>}
 belt_chain_shoulder {parent=belt_chain_shoulder_base, loc=<0 0 0>, rot=<0 0 0 1>}
 belt_chain_shoulder_pnt {parent=belt_chain_shoulder_base, loc=<-0.35 1.72 4.59>, rot=<-0.0001 0 0 1>}
 belt_chain_back_base {parent=spine_05, loc=<-0.8753 -14.1446 -1.1527>, rot=<0.5286 -0.4727 0.4884 -0.5085>}
 belt_chain_back {parent=belt_chain_back_base, loc=<0 0 0>, rot=<0 0 0 1>}
 belt_chain_fr_base {parent=spine_05, loc=<8.1205 9.1012 -7.6087>, rot=<-0.0623 -0.1356 -0.4837 -0.8624>}
 belt_chain_fr {parent=belt_chain_fr_base, loc=<0 0 0>, rot=<0 0 0 1>}
 clavicle_l_2 {parent=spine_05, loc=<4.3575 -0.1466 -2.3223>, rot=<0.0465 -0.7289 -0.009 -0.6829>}
 buckle_pnt {parent=spine_05, loc=<5.34 9.65 -11.6>, rot=<0 0 0 1>}
 belt_chain_base_3 {parent=spine_05, loc=<14.936 -1.1948 -9.752>, rot=<0.006 -0.1491 -0.0397 -0.988>}
 belt_chain_3 {parent=belt_chain_base_3, loc=<0 0 0>, rot=<0 0 0 1>}
 belt_chain_base_6 {parent=spine_05, loc=<1.2946 -13.5515 -0.2148>, rot=<0.5286 -0.4727 0.4884 -0.5085>}
 belt_chain_6 {parent=belt_chain_base_6, loc=<0 0 0>, rot=<0 0 0 1>}
 belt_chain_base_1 {parent=spine_05, loc=<6.8852 9.8936 -7.248>, rot=<-0.158 -0.1886 -0.4656 -0.8501>}
 belt_chain_1 {parent=belt_chain_base_1, loc=<0 0 0>, rot=<0 0 0 1>}
 belt_chain_base_2 {parent=spine_05, loc=<11.6554 5.9408 -9.0441>, rot=<-0.0602 -0.0964 -0.2652 -0.9575>}
 belt_chain_2 {parent=belt_chain_base_2, loc=<0 0 0>, rot=<0 0 0 1>}
 belt_chain_base_4 {parent=spine_05, loc=<12.8743 -7.5193 -7.8475>, rot=<0.1009 -0.1397 0.3628 -0.9158>}
 belt_chain_4 {parent=belt_chain_base_4, loc=<0 0 0>, rot=<0 0 0 1>}
 belt_chain_base_5 {parent=spine_05, loc=<8.8384 -10.7665 -5.4701>, rot=<0.2162 -0.1371 0.4944 -0.8307>}
 belt_chain_5 {parent=belt_chain_base_5, loc=<0 0 0>, rot=<0 0 0 1>}
 belt_chain_base_0 {parent=spine_05, loc=<-0.5878 13.4182 -3.6476>, rot=<-0.1686 -0.1834 -0.5652 -0.7864>}
 belt_chain_0 {parent=belt_chain_base_0, loc=<0 0 0>, rot=<0 0 0 1>}
 deform_pec_aim_base_l {parent=spine_04, loc=<7.8732 12.6388 -2.4123>, rot=<0.0102 -0.4602 0.3979 -0.7936>}
 deform_pec_aim_l {parent=deform_pec_aim_base_l, loc=<0 0 0>, rot=<0 0 0 1>}
 deform_pec_base_l {parent=deform_pec_aim_l, loc=<10.766 0 0>, rot=<0 0 0 1>}
 deform_pec_l {parent=deform_pec_base_l, loc=<0.055 0 0>, rot=<0 0 0 1>}
 deform_pec_belt {parent=deform_pec_l, loc=<0 0 0>, rot=<0 0 0 1>}
 deform_pec_aim_base_r {parent=spine_04, loc=<7.8727 12.6388 2.4123>, rot=<-0.4602 -0.0102 0.7936 0.3979>}
 deform_pec_aim_r {parent=deform_pec_aim_base_r, loc=<0 0 0>, rot=<0 0 0 1>}
 deform_pec_base_r {parent=deform_pec_aim_r, loc=<-10.7662 -0.0006 -0.0007>, rot=<0 0 0 1>}
 deform_pec_r {parent=deform_pec_base_r, loc=<0 0 0>, rot=<0 0 0 1>}
 armpit_l_base {parent=spine_04, loc=<8.2129 1.7027 -17.3855>, rot=<0.4811 -0.4811 -0.5182 -0.5182>}
 armpit_l {parent=armpit_l_base, loc=<0 0 0>, rot=<0 0 0 1>}
 armpit_r_base {parent=spine_04, loc=<8.2126 1.7027 17.3855>, rot=<-0.4811 0.4811 -0.5182 -0.5182>}
 armpit_r {parent=armpit_r_base, loc=<0 0 0>, rot=<0 0 0 1>}
 base_belt_for_r {parent=spine_03, loc=<2.3158 10.8351 9.3269>, rot=<0.4936 0.4319 0.692 0.3016>}
 belt_for_r {parent=base_belt_for_r, loc=<0 0 0>, rot=<0 0 0 1>}
 base_belt_for_l {parent=spine_03, loc=<2.3163 10.8351 -9.3269>, rot=<-0.3016 -0.692 -0.4319 -0.4936>}
 belt_for_l {parent=base_belt_for_l, loc=<0 0 0>, rot=<0 -1 0 0>}
 belt_chain_B_base {parent=spine_03, loc=<3.073 12.9581 7.8526>, rot=<-0.3045 -0.2587 -0.7418 -0.5386>}
 belt_chain_B {parent=belt_chain_B_base, loc=<0 0 0>, rot=<0 0 0 1>}
 belt_chain_A_base {parent=spine_03, loc=<-0.2961 4.7295 13.7836>, rot=<0.617 0.1003 0.7726 0.1114>}
 belt_chain_A {parent=belt_chain_A_base, loc=<0 0 0>, rot=<0 0 0 1>}
 base_belt_for_up_l {parent=spine_03, loc=<6.975 11.9459 -9.8316>, rot=<-0.3016 -0.692 -0.4319 -0.4936>}
 belt_for_up_l {parent=base_belt_for_up_l, loc=<0 0 0>, rot=<0 0 0 1>}
 base_belt_for_up_r {parent=spine_03, loc=<6.9748 11.9459 9.8316>, rot=<0.4936 0.4319 0.692 0.3016>}
 belt_for_up_r {parent=base_belt_for_up_r, loc=<0 0 0>, rot=<0 0 0 -1>}
 base_belt_chain_7 {parent=spine_03, loc=<8.7862 -12.6179 8.4657>, rot=<0.5592 -0.412 0.5608 -0.4507>}
 belt_chain_7 {parent=base_belt_chain_7, loc=<0.3898 0 0>, rot=<0 0 0 1>}
 base_pcoket_A_r {parent=spine_01, loc=<0.1001 1.2564 15.5932>, rot=<0.6943 0.134 0.6983 0.1115>}
 pcoket_A_r {parent=base_pcoket_A_r, loc=<0 0 0>, rot=<0 0 0 1>}
 base_pcoket_B_r {parent=spine_01, loc=<-0.6033 8.8703 10.3854>, rot=<0.6413 0.298 0.6505 0.2772>}
 pcoket_B_r {parent=base_pcoket_B_r, loc=<0 0 0>, rot=<0 0 0 1>}
 base_pcoket_B_l {parent=spine_01, loc=<-0.5907 8.8624 -10.3737>, rot=<-0.2772 -0.6505 -0.298 -0.6413>}
 pcoket_B_l {parent=base_pcoket_B_l, loc=<0 0 0>, rot=<0 0 0 1>}
 base_pcoket_A_l {parent=spine_01, loc=<0.1021 1.2714 -15.5939>, rot=<-0.1115 -0.6983 -0.134 -0.6943>}
 pcoket_A_l {parent=base_pcoket_A_l, loc=<0 0 0>, rot=<0 0 0 1>}
 base_pocket_mid {parent=spine_01, loc=<-1.4103 10.7745 0>, rot=<0.4919 0.508 0.508 0.4919>}
 pocket_mid {parent=base_pocket_mid, loc=<0 0 0>, rot=<0 0 0 1>}
 thigh_l {parent=pelvis, loc=<-2.5693 -0.0109 -7.7501>, rot=<0.0012 0.0572 0.0155 -0.9982>}
 calf_l {parent=thigh_l, loc=<-39.0484 0 0>, rot=<0.0006 -0.0161 -0.0391 -0.9991>}
 foot_l {parent=calf_l, loc=<-42.561 0.1077 0.4292>, rot=<0.0418 -0.0455 0.0343 -0.9975>}
 ball_l {parent=foot_l, loc=<-6.1599 13.8264 0.6806>, rot=<0.0001 -0.0001 -0.7063 -0.7079>}
 foot_l_deform_base {parent=foot_l, loc=<0.5041 -0.0921 0.1615>, rot=<0.5234 0.4729 0.5208 0.4807>}
 foot_l_deform_for_base {parent=foot_l_deform_base, loc=<5.5 0 0>, rot=<0 0 0 1>}
 foot_l_deform_for {parent=foot_l_deform_for_base, loc=<0 0 0>, rot=<0 0 0 1>}
 foot_l_deform_back_base {parent=foot_l_deform_base, loc=<-4.3 0 0>, rot=<0 -1 0 0>}
 foot_l_deform_back {parent=foot_l_deform_back_base, loc=<0 0 0>, rot=<0 0 0 1>}
 calf_twist_02_l {parent=calf_l, loc=<-18.8064 -0.0967 0.0397>, rot=<0.0237 0.0095 -0.0035 -0.9997>}
 calf_twist_01_l {parent=calf_l, loc=<-35.813 -0.0967 0.0397>, rot=<0.0237 0.0095 -0.0035 -0.9997>}
 deform_knee_in_base_l {parent=calf_l, loc=<1.298 -5.2957 -0.0496>, rot=<0 0 0 -1>}
 deform_knee_in_l {parent=deform_knee_in_base_l, loc=<0 0 0>, rot=<0 0 0 1>}
 deform_knee_out_base_l {parent=calf_l, loc=<0 4 0>, rot=<0 0 0 -1>}
 deform_knee_out_l {parent=deform_knee_out_base_l, loc=<0 0 0>, rot=<0 0 0 1>}
 thigh_twist_01_l {parent=thigh_l, loc=<-21.6303 0 0>, rot=<0.0013 0 0 -1>}
 deform_groin_base_l {parent=thigh_l, loc=<-2.1908 8.1142 -1.7442>, rot=<0 0 0 -1>}
 deform_groin_l {parent=deform_groin_base_l, loc=<0 0 0>, rot=<0 0 0 1>}
 thigh_r {parent=pelvis, loc=<-2.5691 -0.0109 7.7501>, rot=<0.0572 -0.0012 0.9982 0.0155>}
 calf_r {parent=thigh_r, loc=<39.0487 0 0>, rot=<0.0006 -0.0161 -0.0391 -0.9991>}
 foot_r {parent=calf_r, loc=<42.5612 -0.1077 -0.4292>, rot=<0.0418 -0.0455 0.0343 -0.9975>}
 ball_r {parent=foot_r, loc=<6.1599 -13.8264 -0.6806>, rot=<0.0001 -0.0001 -0.7063 -0.7079>}
 foot_r_deform_base {parent=foot_r, loc=<-0.5042 0.092 -0.1615>, rot=<-0.5208 0.4807 0.5234 -0.4729>}
 foot_r_deform_for_base {parent=foot_r_deform_base, loc=<5.5 0 0>, rot=<0 0 0 1>}
 foot_r_deform_for {parent=foot_r_deform_for_base, loc=<0 0 0>, rot=<0 0 0 1>}
 foot_r_deform_back_base {parent=foot_r_deform_base, loc=<-4.3 0 0>, rot=<0 -1 0 0>}
 foot_r_deform_back {parent=foot_r_deform_back_base, loc=<0 0 0>, rot=<0 0 0 1>}
 calf_twist_02_r {parent=calf_r, loc=<18.8064 0.0967 -0.0397>, rot=<0.0237 0.0095 -0.0035 -0.9997>}
 calf_twist_01_r {parent=calf_r, loc=<35.8132 0.0967 -0.0397>, rot=<0.0237 0.0095 -0.0035 -0.9997>}
 deform_knee_in_base_r {parent=calf_r, loc=<-1.298 5.2957 0.0496>, rot=<0 0 0 -1>}
 deform_knee_in_r {parent=deform_knee_in_base_r, loc=<0 0 0>, rot=<0 0 0 1>}
 deform_knee_out_base_r {parent=calf_r, loc=<0 -4 0>, rot=<0 0 0 -1>}
 deform_knee_out_r {parent=deform_knee_out_base_r, loc=<0 0 0>, rot=<0 0 0 1>}
 thigh_twist_01_r {parent=thigh_r, loc=<21.6303 0 0>, rot=<0.0013 0 0 -1>}
 base_holster {parent=thigh_twist_01_r, loc=<-7.4146 -0.5623 9.3457>, rot=<-0.0266 0.6837 0.0112 -0.7292>}
 holster {parent=base_holster, loc=<0 0 0>, rot=<0 0 0 1>}
 deform_groin_base_r {parent=thigh_r, loc=<2.1909 -8.1142 1.7442>, rot=<0 0 0 1>}
 deform_groin_r {parent=deform_groin_base_r, loc=<0 0 0>, rot=<0 0 0 1>}
 deform_glute_base_l {parent=pelvis, loc=<-8.8064 -8.8222 -8.4287>, rot=<0.0275 0.0501 -0.4472 -0.8926>}
 deform_glute_l {parent=deform_glute_base_l, loc=<0 0 0>, rot=<0 0 0 1>}
 deform_glute_base_r {parent=pelvis, loc=<-8.8062 -8.8222 8.4287>, rot=<0.0501 -0.0275 0.8926 -0.4472>}
 deform_glute_r {parent=deform_glute_base_r, loc=<0 0 0>, rot=<0 0 0 1>}
 attach {parent=root, loc=<0 0 0>, rot=<0 0 0 1>}
 ik_hand_root {parent=root, loc=<0 0 0>, rot=<0 0 0 1>}
 ik_hand_gun {parent=ik_hand_root, loc=<-51.9729 4.0884 102.6935>, rot=<0.6084 0.1793 0.4613 -0.6205>}
 ik_hand_r {parent=ik_hand_gun, loc=<0 0 0>, rot=<0 0 0 1>}
 ik_hand_l {parent=ik_hand_gun, loc=<53.0286 82.1782 35.2051>, rot=<-0.5102 -0.7906 -0.3387 0>}
 ik_foot_root {parent=root, loc=<0 0 0>, rot=<0 0 0 1>}
 ik_foot_l {parent=ik_foot_root, loc=<15.2608 -4.4176 9.4433>, rot=<0.0357 0.7045 0.0284 -0.7082>}
 ik_foot_r {parent=ik_foot_root, loc=<-15.2609 -4.4176 9.4431>, rot=<0.7082 0.0284 -0.7045 0.0357>}
]
variables [is moving {type=double, default=0}]
