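# Post-pose Animation Blueprint for the M_MED_CymbalBar male character body, layering its Control Rig over the attached parent mesh's pose.
# The AnimGraph copies the parent pose into the rig and passes is_moving, a clamped remap of horizontal pawn velocity.
# Each update also runs the nativized gravity override state function on per-movement-state wind and gravity presets, storing a Gravity Override vector that this AnimGraph never reads.
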AnimBlueprint /BRCosmetics/Characters/Player/Male/Medium/Bodies/M_MED_CymbalBar/Meshes/M_MED_CymbalBar_AnimBP
target_skeleton=&/M_MED_CymbalBar_Skeleton
graphs [
 AnimGraph {nodes [
   n1 = Control Rig!(Source=Copy Pose From Mesh() {bUseAttachedParent=true}, Alpha, is moving:double=Get is_moving() -> (is_moving:double)) -> (Pose) {controlRigAssetReference {blueprintRigClass=&/M_MED_CymbalBar_CtrlRig.M_MED_CymbalBar_CtrlRig_C}}
   Output Pose(Result=@n1) {layerGroup=None}
  ]}
 EventGraph {nodes [
   Event BlueprintUpdateAnimation()
   > n1 = "Gravity Override State Function (Nativized)"!(
    execute
    CharacterPartAnimInstance:Custom Character Part Anim Instance Object Reference=Self-Reference()
    GravityOverrideParameterSet:Fort Gravity Override Parameter Set Structure=Get gravity override params() -> (gravity override params:Fort Gravity Override Parameter Set Structure)
    RigidBodyAnimNodeParameterSet:Fort Rigid Body Anim Node Parameter Set Structure=Get RBAN Params() -> (RBAN Params:Fort Rigid Body Anim Node Parameter Set Structure)
    DebugDraw:bool=false
    GameTimeInSeconds:float=GetGameTimeInSeconds()
    DebugJointName:Name=root
    StateLogicLayerName:Name=none
   ) -> (
    then
    OutGravityOverride:Vector
    OutCurrentState:Name=None
    OutGravityOverrideParameters:Fort Gravity Override Parameters Structure
    OutRigidBodyAnimNodeParameters:Fort Rigid Body Anim Node Parameters Structure
   )
   > Set Gravity Override(Gravity Override=@n1)
   n2 = Get!() -> (PawnVelocity_X:double, PawnVelocity_Y:double, PawnVelocity_Z:double)
   > Set is_moving(is_moving:double=MapRangeClamped(Value=Vector Length!(A_X:double=@n2, A_Y:double=@n2.PawnVelocity_Y, A_Z:double=0) -> (ReturnValue), InRangeA=0, InRangeB=150)) -> (Output_Get:double)
  ]}
]
variables [
 Gravity Override {type=Vector}
 gravity override params {
  type=FortGravityOverrideParameterSet
  default=(Skydive_GoingUp=(GravityOverride=(WindFrequency=(X=1.500000,Y=7.000000,Z=8.000000),WindAmplitude=(X=200.000000,Y=-500.000000,Z=400.000000),GravityOverride=(X=0.000000,Y=-300.000000,Z=1500.000000))),Skydive_FloatingDown=(GravityOverride=(WindFrequency=(X=1.500000,Y=7.000000,Z=8.000000),WindAmplitude=(X=200.000000,Y=-500.000000,Z=400.000000),WindOffset=(X=0.000000,Y=-600.000000,Z=400.000000),GravityOverride=(X=0.000000,Y=-700.000000,Z=600.000000))),Skydive_DivingDown=(GravityOverride=(WindFrequency=(X=1.500000,Y=12.000000,Z=10.000000),WindAmplitude=(X=50.000000,Y=300.000000,Z=100.000000),GravityOverride=(X=-1500.000000,Y=0.000000,Z=0.000000))),Skydive_Parachuting=(GravityOverride=(WindFrequency=(X=1.500000,Y=7.000000,Z=8.000000),WindAmplitude=(X=200.000000,Y=-500.000000,Z=400.000000),GravityOverride=(X=0.000000,Y=-800.000000,Z=-980.000000))),OnGround_Standing=(GravityOverride=(GravityOverride=(X=0.000000,Y=0.000000,Z=-980.000000))),OnGround_Moving=(GravityOverride=(GravityOverride=(X=0.000000,Y=0.000000,Z=-980.000000))),OnGround_Moving_HighRotationRate=(GravityOverride=(GravityOverride=(X=0.000000,Y=0.000000,Z=-980.000000))),OnGround_Crouch_Moving=(GravityOverride=(GravityOverride=(X=0.000000,Y=0.000000,Z=-980.000000))),EmoteOrMelee=(GravityOverride=(GravityOverride=(X=0.000000,Y=0.000000,Z=-980.000000))),Swimming=(GravityOverride=(GravityOverride=(X=-800.000000,Y=0.000000,Z=0.000000))),Front_End=(GravityOverride=(GravityOverride=(X=0.000000,Y=0.000000,Z=-980.000000))),DBNO=(GravityOverride=(GravityOverride=(X=-980.000000,Y=400.000000,Z=0.000000))),DBNO_Carried=(GravityOverride=(GravityOverride=(X=-980.000000,Y=500.000000,Z=-500.000000))),OnGround_Falling=(GravityOverride=(GravityOverride=(X=0.000000,Y=0.000000,Z=-980.000000))),Ziplining=(GravityOverride=(GravityOverride=(X=0.000000,Y=0.000000,Z=-980.000000))),Skydive_Surfing=(GravityOverride=(WindFrequency=(X=1.500000,Y=7.000000,Z=8.000000),WindAmplitude=(X=200.000000,Y=-500.000000,Z=400.000000),GravityOverride=(X=500.000000,Y=-300.000000,Z=-980.000000))),Riding_Standing=(GravityOverride=(GravityOverride=(X=0.000000,Y=-350.000000,Z=-980.000000))),Riding_Moving=(GravityOverride=(GravityOverride=(X=0.000000,Y=-450.000000,Z=-450.000000))),Riding_Moving_Targeting=(GravityOverride=(GravityOverride=(X=0.000000,Y=-200.000000,Z=-980.000000))),Tactical_Sprint=(GravityOverride=(GravityOverride=(X=0.000000,Y=-450.000000,Z=-500.000000))),Rail_Grinding=(GravityOverride=(GravityOverride=(X=0.000000,Y=0.000000,Z=-980.000000))),)
 }
 is_moving {type=number}
 RBAN Params {
  type=FortRigidBodyAnimNodeParameterSet
  default=(Skydive_GoingUp=(RigidBodyAnimNode=()),Skydive_FloatingDown=(RigidBodyAnimNode=()),Skydive_DivingDown=(RigidBodyAnimNode=()),Skydive_Parachuting=(RigidBodyAnimNode=()),OnGround_Standing=(RigidBodyAnimNode=()),OnGround_Moving=(RigidBodyAnimNode=()),OnGround_Moving_HighRotationRate=(RigidBodyAnimNode=()),OnGround_Crouch_Moving=(RigidBodyAnimNode=()),EmoteOrMelee=(RigidBodyAnimNode=()),Swimming=(RigidBodyAnimNode=()),Front_End=(RigidBodyAnimNode=()),DBNO=(RigidBodyAnimNode=()),DBNO_Carried=(RigidBodyAnimNode=()),OnGround_Falling=(RigidBodyAnimNode=()),Ziplining=(RigidBodyAnimNode=()),Skydive_Surfing=(RigidBodyAnimNode=()),Riding_Standing=(RigidBodyAnimNode=()),Riding_Moving=(RigidBodyAnimNode=()),Riding_Moving_Targeting=(RigidBodyAnimNode=()),Tactical_Sprint=(RigidBodyAnimNode=()),Rail_Grinding=(RigidBodyAnimNode=()),)
 }
]
